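# Niagara head effect for the F_MED_Sunshine Eclipse cosmetic: galaxy, smoke and spark sprites placed on skeletal mesh bones or sockets.
# World emitters spawn by distance moved, stopping while skydiving or in vehicles, and drift with curl noise and inherited movement; galaxy and smoke scale by time of day.
# Attached hair and top sparks spawn steadily, resampling the mesh every update.
# Sparks have dual renderers plus dimmer, larger secondary attributes; bisHeadHidden drives every VisibilityTag.
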NiagaraSystem /BRCosmetics/Characters/Player/Female/Medium/Bodies/F_MED_Sunshine/FX/NS_Eclipse_Head
dependencies {
 usedDataInterfaces [&NiagaraDataInterfaceColorCurve, &NiagaraDataInterfaceCurve]
 usedDynamicInputs []
 usedModules [&SystemState]
}
emitters {
 Galaxy_World {
  emitterUpdateScript [
   &EmitterState(Life Cycle Mode=::NewEnumerator1, Loop Duration=1, UseLoopDelay=false)
   &SpawnPerUnit(
    Module.Use Max Movement Threshold=true
    Module.Use Movement Tolerance=false
    Spawn Probability:NiagaraFloat=&Multiply_Float(
     A=&MakeCustomFloatFromBool(Bool=$User.IsSkydiving, True Float=0, False Float=1)
     B=&MakeCustomFloatFromBool(Bool=$User.bisInVehicle, True Float=0, False Float=1)
    )
    Spawn Spacing=40
    Use Spawn Probability=true
   )
  ]
  particleSpawnScript [
   &FN_SetSpawnGroup(Number of Spawn Groups=5)
   &InitializeParticle(
    Color=<rgba 0.72 0.72 0.8 0.15>
    Lifetime Max=0.37
    Lifetime Min=0.42
    Lifetime Mode=::NewEnumerator1
    Sprite Rotation Mode=::NewEnumerator2
    Sprite Size=&Multiply_Vector2D_ByFloat(Float=&RandomRangeFloat(Minimum=7, Maximum=8), Vector2D=<2.5 2.5>)
    Sprite Size Mode=::NewEnumerator4
    UsePositionOffset=false
    Write Material Random=true
   )
   &SphereLocation(
    Hemisphere Y=true
    MaskSpawnGroup=false
    Override Seed=false
    Sphere Radius=2
    Surface Expansion Mode=::NewEnumerator1
    Surface Only=true
    UseNonUniformScale=false
    UseOffset=true
    UseSphereOrientation=false
   )
   &SkeletalMeshLocation(
    Bone / Socket Index=$Particles.SpawnGroup
    Bone Sampling Mode=::NewEnumerator4
    MaskSpawnGroup=false
    Sampled Position Coordinate Space=::World
    Sampled Position Offset=<0 0 10>
    Sampled Position Offset Coordinate Space:ENiagaraSkelMeshTransforms=::NewEnumerator1
    Skeletal Mesh=$User.Skeletal Mesh
   ) {module=SkeletalMeshLocation001}
  ]
  particleUpdateScript [
   &ParticleState()
   &CurlNoiseForce(
    Noise Strength=&FloatFromCurve(
     FloatCurve=DataInterface() {propertyValues {Curve {keys [
         {arriveTangent=-1.3787, interpMode=RCIM_Cubic, leaveTangent=-1.3787, tangentMode=RCTM_User, time=0, value=1}
         {interpMode=RCIM_Cubic, tangentMode=RCTM_User, time=1, value=0}
        ]}}}
     Scale Curve=100
    )
    Pan Noise=false
   )
   &InheritSourceMovement(
    Applied Position Falloff Curve=DataInterface() {propertyValues {Curve {keys []}}}
    Applied Position Scale=<0.95 0.95 0.95>
    ApplyForce=false
    ApplyPosition=true
    ApplyPositionFalloff=true
    ApplyPositionFalloffCurve=false
   )
   &Drag(Drag=2, Use Linear Drag=true, Use Rotational Drag=false)
   &SolveForcesAndVelocity(Clamp Velocity=false, Limit Acceleration=false)
   &ScaleColor(
    Scale RGB=&VectorFromFloat(Value=&FNTimeOfDayScalar(Night Scale=0.5, Morning Scale=0.2, Evening Scale=0.8))
    ScaleA=true
    ScaleRGB=true
   )
   &ScaleSpriteSize(Scale Factor=&Vector2DFromFloat(Value=&FloatFromCurve(FloatCurve=DataInterface() {propertyValues {Curve {keys [
          {arriveTangent=2.3875, interpMode=RCIM_Cubic, leaveTangent=2.3875, tangentMode=RCTM_User, time=0, value=0}
          {arriveTangent=0.5, interpMode=RCIM_Cubic, leaveTangent=0.5, time=0.3, value=1}
          {arriveTangent=-2.3244, interpMode=RCIM_Cubic, leaveTangent=-2.3244, tangentMode=RCTM_User, time=1, value=0.5}
         ]}}})))
   &CameraOffset(Camera Offset Amount=-2)
   &DynamicMaterialParameters(
    Index 0 Param 1=&RandomRangeFloat(Minimum=0, Maximum=1)
    Index 0 Param 2=&RandomRangeFloat(Minimum=0, Maximum=1)
    Module.Param0WriteEnabled=true
    Module.Param1WriteEnabled=true
    Module.Param2WriteEnabled=true
    Module.Param3WriteEnabled=true
   )
   /SetVariables(Particles.VisibilityTag=&MakeCustomIntFromBool(Bool=$User.bisHeadHidden, True INT=1, False INT=0)) {isSetParameters=true, module=SetVariables_#1}
  ]
  rendererClasses [&NiagaraSpriteRendererProperties]
  renderers [{rendererClass=&NiagaraSpriteRendererProperties}]
 }
 Smoke_World {
  emitterUpdateScript [
   &EmitterState(Life Cycle Mode=::NewEnumerator1, Loop Duration=1, UseLoopDelay=false)
   &SpawnPerUnit(
    Module.Use Max Movement Threshold=true
    Module.Use Movement Tolerance=false
    Spawn Probability:NiagaraFloat=&Multiply_Float(
     A=&MakeCustomFloatFromBool(Bool=$User.IsSkydiving, True Float=0, False Float=1)
     B=&MakeCustomFloatFromBool(Bool=$User.bisInVehicle, True Float=0, False Float=1)
    )
    Spawn Spacing=60
    Use Spawn Probability=true
   )
  ]
  particleSpawnScript [
   &FN_SetSpawnGroup(Number of Spawn Groups=5)
   &InitializeParticle(
    Color=<rgba 1 0.8 1 1>
    Lifetime Max=0.4
    Lifetime Min=0.35
    Lifetime Mode=::NewEnumerator1
    Sprite Rotation Mode=::NewEnumerator2
    Sprite Size=&Multiply_Vector2D_ByFloat(Float=&RandomRangeFloat(Minimum=15, Maximum=18), Vector2D=<2.5 2.5>)
    Sprite Size Mode=::NewEnumerator4
    UsePositionOffset=false
    Write Material Random=true
   )
   &SphereLocation(
    Hemisphere Y=true
    MaskSpawnGroup=false
    Override Seed=false
    Sphere Radius=2
    Surface Expansion Mode=::NewEnumerator1
    Surface Only=true
    UseNonUniformScale=false
    UseOffset=true
    UseSphereOrientation=false
   )
   &SkeletalMeshLocation(
    Bone / Socket Index=$Particles.SpawnGroup
    Bone Sampling Mode=::NewEnumerator4
    MaskSpawnGroup=false
    Sampled Position Coordinate Space=::World
    Sampled Position Offset=<0 0 10>
    Sampled Position Offset Coordinate Space:ENiagaraSkelMeshTransforms=::NewEnumerator1
    Skeletal Mesh=$User.Skeletal Mesh
   ) {module=SkeletalMeshLocation001}
  ]
  particleUpdateScript [
   &ParticleState()
   &CurlNoiseForce(
    Noise Strength=&FloatFromCurve(
     FloatCurve=DataInterface() {propertyValues {Curve {keys [
         {arriveTangent=-1.3787, interpMode=RCIM_Cubic, leaveTangent=-1.3787, tangentMode=RCTM_User, time=0, value=1}
         {interpMode=RCIM_Cubic, tangentMode=RCTM_User, time=1, value=0}
        ]}}}
     Scale Curve=100
    )
    Pan Noise=false
   )
   &InheritSourceMovement(
    Applied Position Falloff Curve=DataInterface() {propertyValues {Curve {keys []}}}
    Applied Position Scale=<0.9 0.9 0.9>
    ApplyForce=false
    ApplyPosition=true
    ApplyPositionFalloff=true
    ApplyPositionFalloffCurve=false
   )
   &Drag(Use Linear Drag=true, Use Rotational Drag=false)
   &SolveForcesAndVelocity(Clamp Velocity=false, Limit Acceleration=false)
   &Color(
    Color=&ColorFromCurve(ColorCurve=DataInterface() {propertyValues {
       AlphaCurve {keys [<at 0 1 RCIM_Cubic>,<at 0.9977 1 RCIM_Cubic>]}
       BlueCurve {keys [
         {arriveTangent=1.0606, interpMode=RCIM_Cubic, leaveTangent=1.0606, time=0, value=0.75}
         <at 0.2357 1>
         <at 0.4024 1>
         <at 0.792 1.5>
         {arriveTangent=1.2608, interpMode=RCIM_Cubic, leaveTangent=1.2608, time=1, value=1.7623}
        ]}
       GreenCurve {keys [
         {arriveTangent=0.5303, interpMode=RCIM_Cubic, leaveTangent=0.5303, time=0, value=0.5}
         <at 0.2357 0.625>
         <at 0.4024 0.4583>
         <at 0.792 0.525>
         {arriveTangent=1.8025, interpMode=RCIM_Cubic, leaveTangent=1.8025, time=1, value=0.9}
        ]}
       RedCurve {keys [
         {arriveTangent=-2.1212, interpMode=RCIM_Cubic, leaveTangent=-2.1212, time=0, value=1}
         <at 0.2357 0.5>
         <at 0.4024 0.35>
         <at 0.792 0.6875>
         {arriveTangent=6.3086, interpMode=RCIM_Cubic, leaveTangent=6.3086, time=1, value=2}
        ]}
      }})
    Scale Alpha=&FloatFromCurve(
     FloatCurve=DataInterface() {propertyValues {Curve {keys [
         {interpMode=RCIM_Cubic, tangentMode=RCTM_User, time=-0.0063, value=0.0184}
         {arriveTangent=-0.8483, interpMode=RCIM_Cubic, leaveTangent=-0.8483, tangentMode=RCTM_User, time=0.25, value=1}
         {interpMode=RCIM_Cubic, tangentMode=RCTM_User, time=1.0011, value=0.002}
        ]}}}
     Scale Curve=0.2
    )
    Scale Color=&VectorFromFloat(Value=&FNTimeOfDayScalar(Night Scale=0.5, Morning Scale=0.2, Evening Scale=0.8))
   )
   &ScaleSpriteSize(Scale Factor=&Vector2DFromFloat(Value=&FloatFromCurve(FloatCurve=DataInterface() {propertyValues {Curve {keys [
          {arriveTangent=2.3875, interpMode=RCIM_Cubic, leaveTangent=2.3875, tangentMode=RCTM_User, time=0, value=0}
          {arriveTangent=0.5, interpMode=RCIM_Cubic, leaveTangent=0.5, time=0.6, value=1}
          {arriveTangent=-2.3244, interpMode=RCIM_Cubic, leaveTangent=-2.3244, tangentMode=RCTM_User, time=1, value=0.5}
         ]}}})))
   &CameraOffset(Camera Offset Amount=-2)
   &DynamicMaterialParameters(
    Index 0 Param 1=&RandomRangeFloat(Minimum=0, Maximum=1)
    Index 0 Param 2=&RandomRangeFloat(Minimum=0, Maximum=1)
    Module.Param0WriteEnabled=true
    Module.Param1WriteEnabled=true
    Module.Param2WriteEnabled=true
    Module.Param3WriteEnabled=true
   )
   /SetVariables(Particles.VisibilityTag=&MakeCustomIntFromBool(Bool=$User.bisHeadHidden, True INT=1, False INT=0)) {isSetParameters=true, module=SetVariables_#1}
  ]
  rendererClasses [&NiagaraSpriteRendererProperties]
  renderers [{rendererClass=&NiagaraSpriteRendererProperties}]
 }
 Sparks_Attach_Hair {
  emitterUpdateScript [
   &EmitterState(Life Cycle Mode=::NewEnumerator1, Loop Duration=1, UseLoopDelay=false)
   &SpawnRate(SpawnRate=6, Use Spawn Probability=false)
  ]
  particleSpawnScript [
   &FN_SetSpawnGroup(Number of Spawn Groups=5)
   &InitializeParticle(
    Lifetime Max=1.7
    Lifetime Min=0.5
    Lifetime Mode=::NewEnumerator1
    Sprite Rotation Mode=::NewEnumerator2
    Sprite Size=&Multiply_Vector2D_ByFloat(Float=&RandomRangeFloat(Minimum=1, Maximum=0.3), Vector2D=<1.5 1.5>)
    Sprite Size Mode=::NewEnumerator4
    UsePositionOffset=false
    Write Material Random=true
   )
  ]
  particleUpdateScript [
   &ParticleState()
   &SphereLocation(
    Hemisphere Y=true
    MaskSpawnGroup=false
    Offset=&RandomRangeVector(Minimum=<-8 0 -3>, Maximum=<10 2 -2>)
    Override Seed=false
    Sphere Origin:NiagaraPosition=&MakeVector(X=0, Y=0, Z=0)
    Sphere Radius=1
    Surface Expansion Mode=::NewEnumerator1
    Surface Only=true
    UseNonUniformScale=false
    UseOffset=true
    UseSphereOrientation=false
   )
   &SkeletalMeshLocation(
    Bone / Socket Index=$Particles.SpawnGroup
    Bone Sampling Mode=::NewEnumerator4
    MaskSpawnGroup=false
    Sampled Position Coordinate Space=::World
    Sampled Position Offset=&Add_Vector(
     A=Linked() {linkedVariable {name=Particles.Position, type=NiagaraPosition}}
     B=&MakeVector(
      X=&Lerp_Float(A=0, B=&RandomRangeFloat(Minimum=0, Maximum=1), Alpha=$Particles.NormalizedAge)
      Y=0
      Z=&Lerp_Float(A=0, B=&RandomRangeFloat(Minimum=3, Maximum=1), Alpha=$Particles.NormalizedAge)
     )
    )
    Sampled Position Offset Coordinate Space:ENiagaraSkelMeshTransforms=::NewEnumerator1
    Skeletal Mesh=$User.Skeletal Mesh
   ) {module=SkeletalMeshLocation001}
   &SolveForcesAndVelocity(Clamp Velocity=false, Limit Acceleration=false)
   &Color(
    Color=&ColorFromCurve(ColorCurve=DataInterface() {propertyValues {
       AlphaCurve {keys [<at 0 1 RCIM_Cubic>,<at 0.9977 1 RCIM_Cubic>]}
       BlueCurve {keys [
         {arriveTangent=1.9672, interpMode=RCIM_Cubic, leaveTangent=1.9672, time=0.0747, value=0.3667}
         <at 0.3967 1>
         <at 0.5971 1>
         <at 0.7436 1.5>
         {arriveTangent=0.3488, interpMode=RCIM_Cubic, leaveTangent=0.3488, time=1, value=1.5894}
        ]}
       GreenCurve {keys [
         {arriveTangent=1.553, interpMode=RCIM_Cubic, leaveTangent=1.553, time=0.0747, value=0.05}
         <at 0.3967 0.55>
         <at 0.5971 0.2083>
         <at 0.7436 0.075>
         {arriveTangent=0.0975, interpMode=RCIM_Cubic, leaveTangent=0.0975, time=1, value=0.1}
        ]}
       RedCurve {keys [
         {arriveTangent=-2.7954, interpMode=RCIM_Cubic, leaveTangent=-2.7954, time=0.0747, value=1}
         <at 0.3967 0.1>
         <at 0.5971 0.05>
         <at 0.7436 0.3125>
         {arriveTangent=6.5811, interpMode=RCIM_Cubic, leaveTangent=6.5811, time=1, value=2}
        ]}
      }})
    Scale Alpha=&FloatFromCurve(
     FloatCurve=DataInterface() {propertyValues {Curve {keys [
         {interpMode=RCIM_Cubic, tangentMode=RCTM_User, time=-0.0063, value=0.0184}
         {arriveTangent=-0.8483, interpMode=RCIM_Cubic, leaveTangent=-0.8483, tangentMode=RCTM_User, time=0.1718, value=1.7147}
         {arriveTangent=-5.0978, interpMode=RCIM_Cubic, leaveTangent=-5.0978, tangentMode=RCTM_User, time=0.2818, value=1.2143}
         {arriveTangent=-0.2783, interpMode=RCIM_Cubic, leaveTangent=-0.2783, tangentMode=RCTM_User, time=0.4, value=0.7}
         {arriveTangent=0.3112, interpMode=RCIM_Cubic, leaveTangent=0.3112, tangentMode=RCTM_User, time=0.5407, value=1.3419}
         {arriveTangent=-3.7119, interpMode=RCIM_Cubic, leaveTangent=-3.7119, tangentMode=RCTM_User, time=0.7242, value=0.3749}
         {interpMode=RCIM_Cubic, tangentMode=RCTM_User, time=1.0011, value=0.002}
        ]}}}
     Scale Curve=0.8
    )
    Scale Color=&VectorFromFloat(Value=50)
   )
   /SetVariables(Particles.Color2=&MultiplyLinearColor(Linear Color=$Particles.Color, Scale Factor=&MakeVector4_FromFloat(Float=0.1))) {isSetParameters=true, module=SetVariables_#2}
   &ScaleSpriteSize(Scale Factor=&Vector2DFromFloat(Value=&FloatFromCurve(FloatCurve=DataInterface() {propertyValues {Curve {keys [
          {arriveTangent=7.1136, interpMode=RCIM_Cubic, leaveTangent=7.1136, tangentMode=RCTM_User, time=0, value=0}
          {arriveTangent=1.4654, interpMode=RCIM_Cubic, leaveTangent=1.4654, time=0.18, value=1}
          <at 0.3814 0.5589>
          {arriveTangent=-0.0952, interpMode=RCIM_Cubic, leaveTangent=-0.0952, tangentMode=RCTM_User, time=0.7849, value=0.7663}
          <at 1 0.5 RCIM_Cubic>
         ]}}})))
   /SetVariables(Particles.SpriteSize2=&Multiply_Vector2D_ByFloat(Float=2, Vector2D=$Particles.SpriteSize)) {isSetParameters=true, module=SetVariables_#3}
   &CameraOffset(Camera Offset Amount=2)
   /SetVariables(Particles.VisibilityTag=&MakeCustomIntFromBool(Bool=$User.bisHeadHidden, True INT=1, False INT=0)) {isSetParameters=true, module=SetVariables_#1}
  ]
  rendererClasses [&NiagaraSpriteRendererProperties]
  renderers [
   {rendererClass=&NiagaraSpriteRendererProperties}
   {rendererClass=&NiagaraSpriteRendererProperties, rendererIndex=1}
  ]
 }
 Sparks_Attach_Top {
  emitterUpdateScript [
   &EmitterState(Life Cycle Mode=::NewEnumerator1, Loop Duration=1, UseLoopDelay=false)
   &SpawnRate(SpawnRate=2, Use Spawn Probability=false)
  ]
  particleSpawnScript [&InitializeParticle(
    Lifetime Mode=::NewEnumerator1
    Sprite Rotation Mode=::NewEnumerator2
    Sprite Size=&Multiply_Vector2D_ByFloat(Float=&RandomRangeFloat(Minimum=1.5, Maximum=0.2), Vector2D=<1.5 1.5>)
    Sprite Size Mode=::NewEnumerator4
    UsePositionOffset=false
    Write Material Random=true
   )]
  particleUpdateScript [
   &ParticleState()
   &SkeletalMeshLocation(Bone Sampling Mode=::NewEnumerator4, MaskSpawnGroup=false, Skeletal Mesh=$User.Skeletal Mesh) {module=SkeletalMeshLocation001}
   &SphereLocation(
    Hemisphere X=true
    MaskSpawnGroup=false
    Offset=<6 1 -1>
    Override Seed=false
    Sphere Radius=7
    Surface Only=true
    UseNonUniformScale=false
    UseOffset=true
    UseSphereOrientation=false
   ) {module=SphereLocation001}
   &GravityForce(Gravity=<0 0 50>)
   &SolveForcesAndVelocity(Clamp Velocity=false, Limit Acceleration=false)
   &Color(
    Color=&ColorFromCurve(ColorCurve=DataInterface() {propertyValues {
       AlphaCurve {keys [<at 0 1 RCIM_Cubic>,<at 0.9977 1 RCIM_Cubic>]}
       BlueCurve {keys [
         {arriveTangent=1.8437, interpMode=RCIM_Cubic, leaveTangent=1.8437, time=0.0532, value=0.3667}
         <at 0.3967 1>
         <at 0.5971 1>
         <at 0.7436 1.5>
         {arriveTangent=0.3488, interpMode=RCIM_Cubic, leaveTangent=0.3488, time=1, value=1.5894}
        ]}
       GreenCurve {keys [
         {arriveTangent=1.4555, interpMode=RCIM_Cubic, leaveTangent=1.4555, time=0.0532, value=0.05}
         <at 0.3967 0.55>
         <at 0.5971 0.2083>
         <at 0.7436 0.075>
         {arriveTangent=0.0975, interpMode=RCIM_Cubic, leaveTangent=0.0975, time=1, value=0.1}
        ]}
       RedCurve {keys [
         {arriveTangent=-2.62, interpMode=RCIM_Cubic, leaveTangent=-2.62, time=0.0532, value=1}
         <at 0.3967 0.1>
         <at 0.5971 0.05>
         <at 0.7436 0.3125>
         {arriveTangent=6.5811, interpMode=RCIM_Cubic, leaveTangent=6.5811, time=1, value=2}
        ]}
      }})
    Scale Alpha=&FloatFromCurve(
     FloatCurve=DataInterface() {propertyValues {Curve {keys [
         {interpMode=RCIM_Cubic, tangentMode=RCTM_User, time=-0.0063, value=0.0184}
         {arriveTangent=-0.8483, interpMode=RCIM_Cubic, leaveTangent=-0.8483, tangentMode=RCTM_User, time=0.1718, value=1.7147}
         {arriveTangent=-5.0978, interpMode=RCIM_Cubic, leaveTangent=-5.0978, tangentMode=RCTM_User, time=0.2818, value=1.2143}
         {arriveTangent=-0.2783, interpMode=RCIM_Cubic, leaveTangent=-0.2783, tangentMode=RCTM_User, time=0.4, value=0.7}
         {arriveTangent=0.3112, interpMode=RCIM_Cubic, leaveTangent=0.3112, tangentMode=RCTM_User, time=0.5407, value=1.3419}
         {arriveTangent=-3.7119, interpMode=RCIM_Cubic, leaveTangent=-3.7119, tangentMode=RCTM_User, time=0.7242, value=0.3749}
         {interpMode=RCIM_Cubic, tangentMode=RCTM_User, time=1.0011, value=0.002}
        ]}}}
     Scale Curve=0.8
    )
    Scale Color=&VectorFromFloat(Value=50)
   )
   /SetVariables(Particles.Color2=&MultiplyLinearColor(Linear Color=$Particles.Color, Scale Factor=&MakeVector4_FromFloat(Float=0.1))) {isSetParameters=true, module=SetVariables_#2}
   &ScaleSpriteSize(Scale Factor=&Vector2DFromFloat(Value=&FloatFromCurve(FloatCurve=DataInterface() {propertyValues {Curve {keys [
          {arriveTangent=7.1136, interpMode=RCIM_Cubic, leaveTangent=7.1136, tangentMode=RCTM_User, time=0, value=0}
          {arriveTangent=1.125, interpMode=RCIM_Cubic, leaveTangent=1.125, time=0.2, value=0.85}
          <at 0.4 0.45>
          {arriveTangent=-0.0952, interpMode=RCIM_Cubic, leaveTangent=-0.0952, tangentMode=RCTM_User, time=0.7849, value=0.7663}
          {arriveTangent=-1.238, interpMode=RCIM_Cubic, leaveTangent=-1.238, time=1, value=0.5}
         ]}}})))
   /SetVariables(Particles.SpriteSize2=&Multiply_Vector2D_ByFloat(Float=2, Vector2D=$Particles.SpriteSize)) {isSetParameters=true, module=SetVariables_#3}
   &CameraOffset(Camera Offset Amount=3)
   /SetVariables(Particles.VisibilityTag=&MakeCustomIntFromBool(Bool=$User.bisHeadHidden, True INT=1, False INT=0)) {isSetParameters=true, module=SetVariables_#1}
  ]
  rendererClasses [&NiagaraSpriteRendererProperties]
  renderers [
   {rendererClass=&NiagaraSpriteRendererProperties}
   {rendererClass=&NiagaraSpriteRendererProperties, rendererIndex=1}
  ]
 }
 Sparks_World {
  emitterUpdateScript [
   &EmitterState(Life Cycle Mode=::NewEnumerator1, Loop Duration=1, UseLoopDelay=false)
   &SpawnPerUnit(
    Module.Use Max Movement Threshold=true
    Module.Use Movement Tolerance=false
    Spawn Probability:NiagaraFloat=&Multiply_Float(
     A=&MakeCustomFloatFromBool(Bool=$User.IsSkydiving, True Float=0, False Float=1)
     B=&MakeCustomFloatFromBool(Bool=$User.bisInVehicle, True Float=0, False Float=1)
    )
    Spawn Spacing=60
    Use Spawn Probability=true
   )
  ]
  particleSpawnScript [
   &FN_SetSpawnGroup(Number of Spawn Groups=5)
   &InitializeParticle(
    Lifetime Max=0.4
    Lifetime Min=0.65
    Lifetime Mode=::NewEnumerator1
    Sprite Rotation Mode=::NewEnumerator2
    Sprite Size=&Multiply_Vector2D_ByFloat(Float=&RandomRangeFloat(Minimum=0.9, Maximum=0.25), Vector2D=<2.5 2.5>)
    Sprite Size Mode=::NewEnumerator4
    UsePositionOffset=false
    Write Material Random=true
   )
   &SphereLocation(
    Hemisphere Y=true
    MaskSpawnGroup=false
    Offset=&RandomRangeVector(Minimum=<12 0 0>, Maximum=<8 -5 0>)
    Override Seed=false
    Sphere Origin:NiagaraPosition=&MakeVector(X=0, Y=0, Z=0)
    Sphere Radius=2
    Surface Expansion Mode=::NewEnumerator1
    Surface Only=true
    UseNonUniformScale=false
    UseOffset=true
    UseSphereOrientation=false
   )
   &SkeletalMeshLocation(
    Bone / Socket Index=$Particles.SpawnGroup
    Bone Sampling Mode=::NewEnumerator4
    MaskSpawnGroup=false
    Sampled Position Coordinate Space=::World
    Sampled Position Offset=&Add_Vector(
     A=Linked() {linkedVariable {name=Particles.Position, type=NiagaraPosition}}
     B=&MakeVector(
      X=0
      Y=0
      Z=&Lerp_Float(A=0, B=&RandomRangeFloat(Minimum=-5, Maximum=-2), Alpha=$Particles.NormalizedAge)
     )
    )
    Sampled Position Offset Coordinate Space:ENiagaraSkelMeshTransforms=::NewEnumerator1
    Skeletal Mesh=$User.Skeletal Mesh
   ) {module=SkeletalMeshLocation001}
  ]
  particleUpdateScript [
   &ParticleState()
   &CurlNoiseForce(
    Noise Strength=&FloatFromCurve(
     FloatCurve=DataInterface() {propertyValues {Curve {keys [
         {arriveTangent=-1.3787, interpMode=RCIM_Cubic, leaveTangent=-1.3787, tangentMode=RCTM_User, time=0, value=1}
         {interpMode=RCIM_Cubic, tangentMode=RCTM_User, time=1, value=0}
        ]}}}
     Scale Curve=100
    )
    Pan Noise=false
   )
   &GravityForce(Gravity=&RandomRangeVector(Minimum=<0 0 -100>, Maximum=<1 1 -40>)) {enabled=false}
   &InheritSourceMovement(
    Applied Position Falloff Curve=DataInterface() {propertyValues {Curve {keys []}}}
    Applied Position Scale=<0.9 0.9 0.9>
    ApplyForce=false
    ApplyPosition=true
    ApplyPositionFalloff=true
    ApplyPositionFalloffCurve=false
   )
   &SolveForcesAndVelocity(Clamp Velocity=false, Limit Acceleration=false)
   &Color(
    Color=&ColorFromCurve(ColorCurve=DataInterface() {propertyValues {
       AlphaCurve {keys [<at 0 1 RCIM_Cubic>,<at 0.9977 1 RCIM_Cubic>]}
       BlueCurve {keys [
         {arriveTangent=2.1212, interpMode=RCIM_Cubic, leaveTangent=2.1212, time=0, value=0.5}
         <at 0.2357 1>
         <at 0.4024 1>
         <at 0.792 1.5>
         {arriveTangent=0.4299, interpMode=RCIM_Cubic, leaveTangent=0.4299, time=1, value=1.5894}
        ]}
       GreenCurve {keys [
         {arriveTangent=1.4707, interpMode=RCIM_Cubic, leaveTangent=1.4707, time=0, value=0}
         <at 0.2357 0.3467>
         <at 0.4024 0.2083>
         <at 0.792 0.15>
         {arriveTangent=-0.2403, interpMode=RCIM_Cubic, leaveTangent=-0.2403, time=1, value=0.1}
        ]}
       RedCurve {keys [
         {arriveTangent=-4.1575, interpMode=RCIM_Cubic, leaveTangent=-4.1575, time=0, value=1}
         <at 0.2357 0.02>
         <at 0.4024 0.05>
         <at 0.792 0.375>
         {arriveTangent=7.8107, interpMode=RCIM_Cubic, leaveTangent=7.8107, time=1, value=2}
        ]}
      }})
    Scale Alpha=&FloatFromCurve(
     FloatCurve=DataInterface() {propertyValues {Curve {keys [
         {interpMode=RCIM_Cubic, tangentMode=RCTM_User, time=-0.0063, value=0.0184}
         {arriveTangent=-0.8483, interpMode=RCIM_Cubic, leaveTangent=-0.8483, tangentMode=RCTM_User, time=0.1718, value=1.7147}
         {arriveTangent=-5.0978, interpMode=RCIM_Cubic, leaveTangent=-5.0978, tangentMode=RCTM_User, time=0.2818, value=1.2143}
         {arriveTangent=-0.2783, interpMode=RCIM_Cubic, leaveTangent=-0.2783, tangentMode=RCTM_User, time=0.4294, value=1.1587}
         {arriveTangent=0.3112, interpMode=RCIM_Cubic, leaveTangent=0.3112, tangentMode=RCTM_User, time=0.5407, value=1.3419}
         {arriveTangent=-3.7119, interpMode=RCIM_Cubic, leaveTangent=-3.7119, tangentMode=RCTM_User, time=0.7242, value=0.3749}
         {interpMode=RCIM_Cubic, tangentMode=RCTM_User, time=1.0011, value=0.002}
        ]}}}
     Scale Curve=0.8
    )
    Scale Color=&VectorFromFloat(Value=50)
   )
   /SetVariables(Particles.Color2=&MultiplyLinearColor(Linear Color=$Particles.Color, Scale Factor=&MakeVector4_FromFloat(Float=0.1))) {isSetParameters=true, module=SetVariables_#2}
   &ScaleSpriteSize(Scale Factor=&Vector2DFromFloat(Value=&FloatFromCurve(FloatCurve=DataInterface() {propertyValues {Curve {keys [
          {arriveTangent=7.1136, interpMode=RCIM_Cubic, leaveTangent=7.1136, tangentMode=RCTM_User, time=0, value=0}
          {arriveTangent=1.4654, interpMode=RCIM_Cubic, leaveTangent=1.4654, time=0.18, value=1}
          <at 0.3814 0.5589>
          {arriveTangent=-0.0952, interpMode=RCIM_Cubic, leaveTangent=-0.0952, tangentMode=RCTM_User, time=0.7849, value=0.7663}
          <at 1 0.5 RCIM_Cubic>
         ]}}})))
   /SetVariables(Particles.SpriteSize2=&Multiply_Vector2D_ByFloat(Float=4, Vector2D=$Particles.SpriteSize)) {isSetParameters=true, module=SetVariables_#3}
   &CameraOffset(Camera Offset Amount=-2)
   &KillParticlesInVolume(Invert Volume=true)
   /SetVariables(Particles.VisibilityTag=&MakeCustomIntFromBool(Bool=$User.bisHeadHidden, True INT=1, False INT=0)) {isSetParameters=true, module=SetVariables_#1}
  ]
  rendererClasses [&NiagaraSpriteRendererProperties]
  renderers [
   {rendererClass=&NiagaraSpriteRendererProperties}
   {rendererClass=&NiagaraSpriteRendererProperties, rendererIndex=1}
  ]
 }
}
summary {
 systemName=NS_Eclipse_Head
 emitters [
  {emitterName=Sparks_Attach_Top, rendererClasses [&NiagaraSpriteRendererProperties]}
  {emitterName=Sparks_Attach_Hair, rendererClasses [&NiagaraSpriteRendererProperties]}
  {emitterName=Sparks_World, rendererClasses [&NiagaraSpriteRendererProperties]}
  {emitterName=Smoke_World, rendererClasses [&NiagaraSpriteRendererProperties]}
  {emitterName=Galaxy_World, rendererClasses [&NiagaraSpriteRendererProperties]}
 ]
 userVariables [
  User.bisHeadHidden {type=NiagaraBool}
  User.bisInVehicle {type=NiagaraBool}
  User.IsSkydiving {type=NiagaraBool}
  User.Skeletal Mesh {
   type=NiagaraDataInterfaceSkeletalMesh
   defaultValue {
    struct=NiagaraExt_VariableValue_DataInterface
    value {dataInterface=&self:NiagaraDataInterfaceSkeletalMesh_0}
   }
  }
 ]
}
