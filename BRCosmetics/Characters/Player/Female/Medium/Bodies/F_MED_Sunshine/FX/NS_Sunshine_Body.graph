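# Niagara cosmetic effect for the Sunshine female body, pinned to two skeletal mesh bones or sockets, apparently the feet.
# Persistent glow sprites and rainbow meshes scale with time of day, and the meshes pass smoothed position to their material for motion lag.
# Double-sprite rainbow sparks spawn attached at a steady rate and in the world per distance moved.
# Hidden feet suppress all emitters; skydiving and vehicles stop only sparks.
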NiagaraSystem /BRCosmetics/Characters/Player/Female/Medium/Bodies/F_MED_Sunshine/FX/NS_Sunshine_Body
dependencies {
 usedDataInterfaces [
  &NiagaraDataInterfaceColorCurve
  &NiagaraDataInterfaceArrayFloat3
  &NiagaraDataInterfaceCurve
  &NiagaraDataInterfaceArrayFloat
 ]
 usedDynamicInputs []
 usedModules [&SystemState]
}
emitters {
 Glow {
  emitterUpdateScript [
   &EmitterState(Life Cycle Mode=::NewEnumerator1, Loop Behavior=::NewEnumerator1, Loop Duration Mode=::NewEnumerator1, UseLoopDelay=false)
   &SpawnBurst_Instantaneous(Spawn Count=2, Use Spawn Probability=false, UseLoopCountLimit=false)
  ]
  particleSpawnScript [
   &FN_SetSpawnGroup(Number of Spawn Groups=2)
   &InitializeParticle(Color=<rgba 1 0.6667 0.5 0.4>, Sprite Size Mode=::NewEnumerator3, Uniform Sprite Size=5, UsePositionOffset=false, Write Material Random=true)
  ]
  particleUpdateScript [
   &ParticleState(Kill Particles When Lifetime Has Elapsed=false, Let Infinitely Lived Particles Die When Emitter Deactivates=true)
   &SkeletalMeshLocation(
    Bone / Socket Index=$Particles.SpawnGroup
    Bone Sampling Mode=::NewEnumerator4
    MaskSpawnGroup=false
    Orientation Sampling=::NewEnumerator1
    Sampled Position Offset Coordinate Space:ENiagaraSkelMeshTransforms=::NewEnumerator4
    Skeletal Mesh=$User.Skeletal Mesh
   )
   &CameraOffset(Camera Offset Amount=5)
   &ScaleColor(
    Scale RGB=&VectorFromFloat(Value=&FNTimeOfDayScalar(Night Scale=0.6, Morning Scale=0.2, Evening Scale=0.8))
    ScaleA=true
    ScaleRGB=true
   )
   &ScaleColor(
    Scale Alpha=&MakeCustomFloatFromBool(Bool=$User.bisFootHidden, True Float=0, False Float=1)
    ScaleA=true
    ScaleRGB=false
   ) {module=ScaleColor001}
  ]
  rendererClasses [&NiagaraSpriteRendererProperties]
  renderers [{rendererClass=&NiagaraSpriteRendererProperties}]
 }
 Rianbow_Mesh {
  emitterUpdateScript [
   &EmitterState(Life Cycle Mode=::NewEnumerator1, Loop Behavior=::NewEnumerator1, Loop Duration Mode=::NewEnumerator1, UseLoopDelay=false)
   &SpawnBurst_Instantaneous(Spawn Count=2, Use Spawn Probability=false, UseLoopCountLimit=false)
  ]
  particleSpawnScript [
   &FN_SetSpawnGroup(Number of Spawn Groups=2)
   &InitializeParticle(
    Color=<rgba 0.75 0.875 1.5 1>
    Mesh Scale=<1 0.85 0.8>
    Mesh Scale Mode=::NewEnumerator4
    Sprite Size Mode=::NewEnumerator3
    Uniform Sprite Size=20
    UsePositionOffset=false
    Write Material Random=true
   )
   &InitialMeshOrientation(
    Mesh Orientation Mode=::NewEnumerator2
    Orientation Vector=&SelectVectorFromArray(Direct Array Index=$Particles.SpawnGroup, Vector Selection Array=DataInterface() {propertyValues {FloatData [<1 0 0>, <-1 0 0>]}})
    Rotation=&SelectVectorFromArray(Direct Array Index=$Particles.SpawnGroup, Vector Selection Array=DataInterface() {propertyValues {FloatData [<0 0.25 0>, <0 0.25 0>]}})
    Use Rotation Vector=true
   )
  ]
  particleUpdateScript [
   &ParticleState(Kill Particles When Lifetime Has Elapsed=false, Let Infinitely Lived Particles Die When Emitter Deactivates=true)
   &SkeletalMeshLocation(
    Bone / Socket Index=$Particles.SpawnGroup
    Bone Sampling Mode=::NewEnumerator4
    MaskSpawnGroup=false
    Sampled Position Offset Coordinate Space:ENiagaraSkelMeshTransforms=::NewEnumerator4
    Skeletal Mesh=$User.Skeletal Mesh
   )
   &InterpolateOverTime(Rate Of Change=5, Target Vector Value:NiagaraPosition=$Particles.Position, Variable Type=::NewEnumerator2)
   /SetVariables(Particles.InterpolateDynamic:Vector4f=&MakeFloat4FromVectorAndFloat("Vector (RGB)"=$Output.InterpolateOverTime.Moving Average, "Float (Alpha)"=1.5)) {isSetParameters=true}
   &DynamicMaterialParameters(
    Index 0 Param 4=&RandomRangeFloat(Minimum=-10, Maximum=10)
    Module.Param0WriteEnabled=true
    Module.Param1WriteEnabled=true
    Module.Param2WriteEnabled=true
    Module.Param3WriteEnabled=true
   )
   &ScaleColor(
    Scale Alpha=&FN_FrontEndFloatModifier(Gameplay=1, Front End=3)
    Scale RGB=&VectorFromFloat(Value=&FNTimeOfDayScalar(Night Scale=0.6, Morning Scale=0.2, Day Scale=1.15, Evening Scale=0.8))
    ScaleA=true
    ScaleRGB=true
   )
   &ScaleColor(
    Scale Alpha=&MakeCustomFloatFromBool(Bool=$User.bisFootHidden, True Float=0, False Float=1)
    ScaleA=true
    ScaleRGB=false
   ) {module=ScaleColor001}
  ]
  rendererClasses [&NiagaraMeshRendererProperties]
  renderers [{rendererClass=&NiagaraMeshRendererProperties, rendererIndex=0}]
 }
 Sparks_Attach {
  emitterUpdateScript [
   &EmitterState(Life Cycle Mode=::NewEnumerator1, Loop Duration=1, UseLoopDelay=false)
   &SpawnRate(
    Spawn Probability=&Multiply_Float(
     A=&MakeCustomFloatFromBool(Bool=$User.IsSkydiving, True Float=0, False Float=1)
     B=&MakeCustomFloatFromBool(Bool=$User.bisInVehicle, True Float=0, False Float=1)
    )
    SpawnRate=&MakeCustomFloatFromBool(Bool=$User.bisFootHidden, True Float=0, False Float=10)
    Use Spawn Probability=true
   )
  ]
  particleSpawnScript [
   &FN_SetSpawnGroup(Number of Spawn Groups=2)
   &InitializeParticle(
    Lifetime Max=0.8
    Lifetime Min=2.2
    Lifetime Mode=::NewEnumerator1
    Sprite Rotation Mode=::NewEnumerator2
    Sprite Size Mode=::NewEnumerator3
    Uniform Sprite Size=&FloatFromCurve(
     FloatCurve=DataInterface() {propertyValues {Curve {keys [<at 0 0.2>,<at 0.8 1>,<at 0.9 1.5>]}}}
     CurveIndex=&RandomRangeFloat(Minimum=0, Maximum=1)
     Scale Curve=1.85
    )
    UsePositionOffset=false
    Write Material Random=true
   )
   &SphereLocation(
    MaskSpawnGroup=false
    Override Seed=false
    Sphere Origin:NiagaraPosition=&MakeVector(X=0, Y=0, Z=0)
    Sphere Radius=4
    Surface Expansion Mode=::NewEnumerator1
    Surface Only=true
    UseNonUniformScale=false
    UseOffset=true
    UseSphereOrientation=false
   )
   &AddVelocity(Velocity=&SelectVectorFromArray(Direct Array Index=$Particles.SpawnGroup, Vector Selection Array=DataInterface() {propertyValues {FloatData [<0 -25 20>, <20 -25 20>]}})) {enabled=false}
  ]
  particleUpdateScript [
   &ParticleState()
   &SkeletalMeshLocation(
    Bone / Socket Index=$Particles.SpawnGroup
    Bone Sampling Mode=::NewEnumerator4
    MaskSpawnGroup=false
    Sampled Position Offset=&Add_Vector(
     A=Linked() {linkedVariable {name=Particles.Initial.Position, type=NiagaraPosition}}
     B=&MakeVector(
      X=&SelectFloatFromArray(Direct Array Index=$Particles.SpawnGroup, Float Selection Array=DataInterface() {propertyValues {FloatData [5,-5]}})
      Y=&Lerp_Float(
       A=0
       B=&SelectFloatFromArray(Direct Array Index=$Particles.SpawnGroup, Float Selection Array=DataInterface() {propertyValues {FloatData [-20,20]}})
       Alpha=$Particles.NormalizedAge
      )
      Z=0
     )
    )
    Skeletal Mesh=$User.Skeletal Mesh
   ) {module=SkeletalMeshLocation001}
   &CurlNoiseForce(
    Noise Frequency=20
    Noise Strength=&FloatFromCurve(
     FloatCurve=DataInterface() {propertyValues {Curve {keys [
         {arriveTangent=-1.3787, interpMode=RCIM_Cubic, leaveTangent=-1.3787, tangentMode=RCTM_User, time=0, value=1}
         <at 0.2 0.1>
         {interpMode=RCIM_Cubic, tangentMode=RCTM_User, time=1, value=0}
        ]}}}
     Scale Curve=100
    )
    Pan Noise=true
    Pan Noise Field=<0 0 1>
   )
   &GravityForce(Gravity=&RandomRangeVector(Minimum=<0 0 60>, Maximum=<0 0 45>))
   &InheritVelocity(
    Inherited Velocity Amount Scale=<-0.01 -0.01 -0.01>
    Speed Limit=true
    UseSpeedThreshold=false
    Velocity Source=&TransformVector(Vector=$Engine.Owner.Velocity, Source Space=::World, Destination Space=::Local)
   ) {enabled=false}
   &Drag(Drag=&RandomRangeFloat(Minimum=1, Maximum=0), Use Linear Drag=true, Use Rotational Drag=false) {enabled=false}
   &InheritSourceMovement(ApplyForce=false, ApplyPosition=true, ApplyPositionFalloff=false) {enabled=false}
   &SolveForcesAndVelocity(Clamp Velocity=false, Limit Acceleration=false)
   &Color(
    Color=&ColorFromCurve(ColorCurve=DataInterface() {propertyValues {
       AlphaCurve {keys [<at 0 1 RCIM_Cubic>,<at 0.9977 1 RCIM_Cubic>]}
       BlueCurve {keys [
         {arriveTangent=0.0766, interpMode=RCIM_Cubic, leaveTangent=0.0766, time=0.004, value=0}
         <at 0.1764 0.01>
         <at 0.3718 1>
         <at 0.615 1>
         <at 0.7506 1.5>
         {arriveTangent=0.3586, interpMode=RCIM_Cubic, leaveTangent=0.3586, time=1, value=1.5894}
        ]}
       GreenCurve {keys [
         {arriveTangent=2.5528, interpMode=RCIM_Cubic, leaveTangent=2.5528, time=0.004, value=0.1667}
         <at 0.1764 0.5>
         <at 0.3718 1>
         <at 0.615 0.2083>
         <at 0.7506 0.075>
         {arriveTangent=0.1002, interpMode=RCIM_Cubic, leaveTangent=0.1002, time=1, value=0.1}
        ]}
       RedCurve {keys [
         {arriveTangent=-5.7055, interpMode=RCIM_Cubic, leaveTangent=-5.7055, time=0.004, value=1}
         <at 0.1764 0.255>
         <at 0.3718 0.05>
         <at 0.615 0.05>
         <at 0.7506 0.3125>
         {arriveTangent=6.7652, interpMode=RCIM_Cubic, leaveTangent=6.7652, time=1, value=2}
        ]}
      }})
    Scale Alpha=&FloatFromCurve(
     FloatCurve=DataInterface() {propertyValues {Curve {keys [
         {interpMode=RCIM_Cubic, tangentMode=RCTM_User, time=-0.0063, value=0.0184}
         {arriveTangent=-0.8483, interpMode=RCIM_Cubic, leaveTangent=-0.8483, tangentMode=RCTM_User, time=0.1718, value=1.7147}
         {arriveTangent=-5.0978, interpMode=RCIM_Cubic, leaveTangent=-5.0978, tangentMode=RCTM_User, time=0.2818, value=1.2143}
         {arriveTangent=-0.2783, interpMode=RCIM_Cubic, leaveTangent=-0.2783, tangentMode=RCTM_User, time=0.4, value=0.7}
         {arriveTangent=0.3112, interpMode=RCIM_Cubic, leaveTangent=0.3112, tangentMode=RCTM_User, time=0.5407, value=1.3419}
         {arriveTangent=-3.7119, interpMode=RCIM_Cubic, leaveTangent=-3.7119, tangentMode=RCTM_User, time=0.7242, value=0.3749}
         {interpMode=RCIM_Cubic, tangentMode=RCTM_User, time=1.0011, value=0.002}
        ]}}}
     Scale Curve=0.8
    )
    Scale Color=&VectorFromFloat(Value=50)
   )
   /SetVariables(Particles.Color2=&MultiplyLinearColor(Linear Color=$Particles.Color, Scale Factor=&MakeVector4_FromFloat(Float=0.1))) {isSetParameters=true, module=SetVariables_#1}
   &ScaleSpriteSize(Scale Factor=&Vector2DFromFloat(Value=&FloatFromCurve(FloatCurve=DataInterface() {propertyValues {Curve {keys [
          {arriveTangent=7.1136, interpMode=RCIM_Cubic, leaveTangent=7.1136, tangentMode=RCTM_User, time=0, value=0}
          {arriveTangent=0.099, interpMode=RCIM_Cubic, leaveTangent=0.099, tangentMode=RCTM_User, time=0.2, value=0.9}
          {arriveTangent=0.1511, interpMode=RCIM_Cubic, leaveTangent=0.1511, tangentMode=RCTM_User, time=0.5, value=0.5}
          {arriveTangent=-0.0952, interpMode=RCIM_Cubic, leaveTangent=-0.0952, tangentMode=RCTM_User, time=0.7, value=0.75}
          {arriveTangent=-2.5, interpMode=RCIM_Cubic, leaveTangent=-2.5, time=1, value=0}
         ]}}})))
   /SetVariables(Particles.SpriteSize2=&Multiply_Vector2D_ByFloat(Float=4, Vector2D=$Particles.SpriteSize)) {isSetParameters=true, module=SetVariables_#2}
   &CameraOffset(Camera Offset Amount=1)
   &KillParticlesInVolume(Invert Volume=true)
  ]
  rendererClasses [&NiagaraSpriteRendererProperties]
  renderers [
   {rendererClass=&NiagaraSpriteRendererProperties}
   {rendererClass=&NiagaraSpriteRendererProperties, rendererIndex=1}
  ]
 }
 Sparks_World {
  emitterUpdateScript [
   &EmitterState(Life Cycle Mode=::NewEnumerator1, Loop Duration=1, UseLoopDelay=false)
   &SpawnPerUnit(
    Module.Use Max Movement Threshold=true
    Module.Use Movement Tolerance=false
    Spawn Probability:NiagaraFloat=&Multiply_Float(
     A=&MakeCustomFloatFromBool(Bool=$User.IsSkydiving, True Float=0, False Float=1)
     B=&MakeCustomFloatFromBool(Bool=$User.bisInVehicle, True Float=0, False Float=1)
    )
    Spawn Spacing:NiagaraFloat=&MakeCustomFloatFromBool(Bool=$User.bisFootHidden, True Float=0, False Float=20)
    Use Spawn Probability=true
   )
  ]
  particleSpawnScript [
   &FN_SetSpawnGroup(Number of Spawn Groups=2)
   &InitializeParticle(
    Lifetime Max=0.5
    Lifetime Mode=::NewEnumerator1
    Sprite Rotation Mode=::NewEnumerator2
    Sprite Size Mode=::NewEnumerator3
    Uniform Sprite Size=&FloatFromCurve(
     FloatCurve=DataInterface() {propertyValues {Curve {keys [<at 0 0.2>,<at 0.8 1>,<at 0.9 1.5>]}}}
     CurveIndex=&RandomRangeFloat(Minimum=0, Maximum=1)
     Scale Curve=1.85
    )
    UsePositionOffset=false
    Write Material Random=true
   )
   &SphereLocation(
    MaskSpawnGroup=false
    Override Seed=false
    Sphere Radius=4
    Surface Expansion Mode=::NewEnumerator1
    Surface Only=true
    UseNonUniformScale=false
    UseOffset=true
    UseSphereOrientation=false
   )
   &SkeletalMeshLocation(Bone / Socket Index=$Particles.SpawnGroup, Bone Sampling Mode=::NewEnumerator4, MaskSpawnGroup=false, Skeletal Mesh=$User.Skeletal Mesh) {module=SkeletalMeshLocation001}
  ]
  particleUpdateScript [
   &ParticleState()
   &CurlNoiseForce(
    Noise Frequency=20
    Noise Strength=&FloatFromCurve(
     FloatCurve=DataInterface() {propertyValues {Curve {keys [
         {arriveTangent=-1.3787, interpMode=RCIM_Cubic, leaveTangent=-1.3787, tangentMode=RCTM_User, time=0, value=1}
         <at 0.2 0.1>
         {interpMode=RCIM_Cubic, tangentMode=RCTM_User, time=1, value=0}
        ]}}}
     Scale Curve=100
    )
    Pan Noise=true
    Pan Noise Field=<0 0 1>
   )
   &GravityForce(Gravity=&RandomRangeVector(Minimum=<0 0 -20>, Maximum=<0 0 -20>))
   &InheritSourceMovement(Applied Position Scale=<0.2 0.2 0.2>, ApplyForce=false, ApplyPosition=true, ApplyPositionFalloff=false)
   &SolveForcesAndVelocity(Clamp Velocity=false, Limit Acceleration=false)
   &Color(
    Color=&ColorFromCurve(ColorCurve=DataInterface() {propertyValues {
       AlphaCurve {keys [<at 0 1 RCIM_Cubic>,<at 0.9977 1 RCIM_Cubic>]}
       BlueCurve {keys [
         {arriveTangent=0.0766, interpMode=RCIM_Cubic, leaveTangent=0.0766, time=0.004, value=0}
         <at 0.1764 0.01>
         <at 0.3718 1>
         <at 0.615 1>
         <at 0.7506 1.5>
         {arriveTangent=0.3586, interpMode=RCIM_Cubic, leaveTangent=0.3586, time=1, value=1.5894}
        ]}
       GreenCurve {keys [
         {arriveTangent=2.5528, interpMode=RCIM_Cubic, leaveTangent=2.5528, time=0.004, value=0.1667}
         <at 0.1764 0.5>
         <at 0.3718 1>
         <at 0.615 0.2083>
         <at 0.7506 0.075>
         {arriveTangent=0.1002, interpMode=RCIM_Cubic, leaveTangent=0.1002, time=1, value=0.1}
        ]}
       RedCurve {keys [
         {arriveTangent=-5.7055, interpMode=RCIM_Cubic, leaveTangent=-5.7055, time=0.004, value=1}
         <at 0.1764 0.255>
         <at 0.3718 0.05>
         <at 0.615 0.05>
         <at 0.7506 0.3125>
         {arriveTangent=6.7652, interpMode=RCIM_Cubic, leaveTangent=6.7652, time=1, value=2}
        ]}
      }})
    Scale Alpha=&FloatFromCurve(
     FloatCurve=DataInterface() {propertyValues {Curve {keys [
         {interpMode=RCIM_Cubic, tangentMode=RCTM_User, time=-0.0063, value=0.0184}
         {arriveTangent=-0.8483, interpMode=RCIM_Cubic, leaveTangent=-0.8483, tangentMode=RCTM_User, time=0.1718, value=1.7147}
         {arriveTangent=-5.0978, interpMode=RCIM_Cubic, leaveTangent=-5.0978, tangentMode=RCTM_User, time=0.2818, value=1.2143}
         {arriveTangent=-0.2783, interpMode=RCIM_Cubic, leaveTangent=-0.2783, tangentMode=RCTM_User, time=0.4, value=0.7}
         {arriveTangent=0.3112, interpMode=RCIM_Cubic, leaveTangent=0.3112, tangentMode=RCTM_User, time=0.5407, value=1.3419}
         {arriveTangent=-3.7119, interpMode=RCIM_Cubic, leaveTangent=-3.7119, tangentMode=RCTM_User, time=0.7242, value=0.3749}
         {interpMode=RCIM_Cubic, tangentMode=RCTM_User, time=1.0011, value=0.002}
        ]}}}
     Scale Curve=0.8
    )
    Scale Color=&VectorFromFloat(Value=50)
   )
   /SetVariables(Particles.Color2=&MultiplyLinearColor(Linear Color=$Particles.Color, Scale Factor=&MakeVector4_FromFloat(Float=0.1))) {isSetParameters=true, module=SetVariables_#1}
   &ScaleSpriteSize(Scale Factor=&Vector2DFromFloat(Value=&FloatFromCurve(FloatCurve=DataInterface() {propertyValues {Curve {keys [
          {arriveTangent=7.1136, interpMode=RCIM_Cubic, leaveTangent=7.1136, tangentMode=RCTM_User, time=0, value=0}
          {arriveTangent=0.099, interpMode=RCIM_Cubic, leaveTangent=0.099, tangentMode=RCTM_User, time=0.2, value=0.9}
          {arriveTangent=0.1511, interpMode=RCIM_Cubic, leaveTangent=0.1511, tangentMode=RCTM_User, time=0.5, value=0.5}
          {arriveTangent=-0.0952, interpMode=RCIM_Cubic, leaveTangent=-0.0952, tangentMode=RCTM_User, time=0.7, value=0.75}
          {arriveTangent=-2.5, interpMode=RCIM_Cubic, leaveTangent=-2.5, time=1, value=0}
         ]}}})))
   /SetVariables(Particles.SpriteSize2=&Multiply_Vector2D_ByFloat(Float=4, Vector2D=$Particles.SpriteSize)) {isSetParameters=true, module=SetVariables_#2}
   &CameraOffset(Camera Offset Amount=1)
   &KillParticlesInVolume(Invert Volume=true)
  ]
  rendererClasses [&NiagaraSpriteRendererProperties]
  renderers [
   {rendererClass=&NiagaraSpriteRendererProperties}
   {rendererClass=&NiagaraSpriteRendererProperties, rendererIndex=1}
  ]
 }
}
summary {
 systemName=NS_Sunshine_Body
 emitters [
  {emitterName=Glow, rendererClasses [&NiagaraSpriteRendererProperties]}
  {emitterName=Rianbow_Mesh, rendererClasses [&NiagaraMeshRendererProperties]}
  {emitterName=Sparks_Attach, rendererClasses [&NiagaraSpriteRendererProperties]}
  {emitterName=Sparks_World, rendererClasses [&NiagaraSpriteRendererProperties]}
 ]
 userVariables [
  User.bisFootHidden {type=NiagaraBool}
  User.bIsFrontEnd {type=NiagaraBool}
  User.bIsFrontEndPreview {type=NiagaraBool}
  User.bisInVehicle {type=NiagaraBool}
  User.IsSkydiving {type=NiagaraBool}
  User.Skeletal Mesh {
   type=NiagaraDataInterfaceSkeletalMesh
   defaultValue {
    struct=NiagaraExt_VariableValue_DataInterface
    value {dataInterface=&self:NiagaraDataInterfaceSkeletalMesh_0}
   }
  }
 ]
}
