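# Control Rig for the F_MED_Sunshine female player body that rotates the two back collar bones in response to head rotation.
# It reads the head transform relative to neck_01 and takes the Z Euler angle of its rotation.
# A clamped remap with an inverted source range turns that angle into a Y-axis rotation.
# The same rotation is applied additively in local space to dyn_collar_back_l, then dyn_collar_back_r; no other bone is modified.
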ControlRigBlueprint /BRCosmetics/Characters/Player/Female/Medium/Bodies/F_MED_Sunshine/Meshes/F_MED_Sunshine_Ctrl_Rig
graphs [
 RigVMFunctionLibrary {links [], nodes []}
 RigVMModel {nodes [
   RigUnit_BeginExecution = &RigUnit_BeginExecution()
   n1 = &RigUnit_GetRelativeBoneTransform(Bone=head, Space=neck_01)
   n2 = &RigVMFunction_MathQuaternionFromEuler(
    Euler=(X=0,Y=0,Z=0)
    Euler.Y=&RigVMFunction_MathFloatRemap(
     SourceMaximum=-50
     SourceMinimum=10
     TargetMaximum=-50
     TargetMinimum=0
     bClamp=true
     Value=&RigVMFunction_MathQuaternionToEuler(RotationOrder=ZYX, Value=@n1.Transform.Rotation).Result.Z
    ).Result
   )
   /Script/ControlRig.RigUnit_ModifyBoneTransforms(
    BoneToModify=((Bone="dyn_collar_back_l",Transform=(Rotation=(X=0,Y=0,Z=0,W=1),Translation=(X=0,Y=0,Z=0),Scale3D=(X=1,Y=1,Z=1))))
    Mode=AdditiveLocal
    Weight=1
    WeightMaximum=1
    WeightMinimum=0
    BoneToModify.0.Transform.Rotation=@n2.Result
    ExecutePin=@RigUnit_BeginExecution.ExecutePin
   ) {is=struct RigVMUnitNode}
   > &RigUnit_ModifyBoneTransforms(
    BoneToModify=((Bone="dyn_collar_back_r",Transform=(Rotation=(X=0,Y=0,Z=0,W=1),Translation=(X=0,Y=0,Z=0),Scale3D=(X=1,Y=1,Z=1))))
    Mode=AdditiveLocal
    Weight=1
    WeightMaximum=1
    BoneToModify.0.Transform.Rotation=@n2.Result
   )
  ]}
]
hierarchy [
 root {loc=<0 0 0>, rot=<0 0 0 1>}
 attach {parent=root, loc=<0 0 0>, rot=<0 0 0 1>}
 pelvis {parent=root, loc=<0 -0.5301 98.0571>, rot=<0.0127 0.707 -0.0127 -0.707>}
 spine_01 {parent=pelvis, loc=<6.1208 -0.9328 0>, rot=<0 0 0.034 -0.9994>}
 spine_02 {parent=spine_01, loc=<6.5996 -0.2095 0>, rot=<0 0 0.0238 -0.9997>}
 spine_03 {parent=spine_02, loc=<6.4914 -0.1237 0>, rot=<0 0 0.0149 -0.9999>}
 spine_04 {parent=spine_03, loc=<6.9598 0.0794 0>, rot=<0 0 -0.0132 -0.9999>}
 spine_05 {parent=spine_04, loc=<6.9987 0 0>, rot=<0 0 -0.0015 -1>}
 clavicle_l {parent=spine_05, loc=<5.4831 -0.1053 -3.2946>, rot=<0.0465 -0.7289 -0.009 -0.6829>}
 upperarm_l {parent=clavicle_l, loc=<9.8658 -1.2136 -0.4105>, rot=<0.0164 -0.3245 -0.0269 -0.9454>}
 lowerarm_l {parent=upperarm_l, loc=<26.465 -0.8345 0.259>, rot=<0.0036 0.0161 -0.2298 -0.9731>}
 hand_l {parent=lowerarm_l, loc=<22.7355 -0.3203 -0.6105>, rot=<-0.6784 -0.0223 0.0735 -0.7307>}
 index_metacarpal_l {parent=hand_l, loc=<1.6535 0.0183 -2.96>, rot=<0.0158 -0.0124 -0.0173 -0.9996>}
 index_01_l {parent=index_metacarpal_l, loc=<8.3008 0.0357 -0.0524>, rot=<-0.0044 0.0485 0.0924 -0.9945>}
 index_02_l {parent=index_01_l, loc=<3.3587 -0.3003 0.0452>, rot=<-0.002 0.0027 0.3225 -0.9466>}
 index_03_l {parent=index_02_l, loc=<2.5578 -0.123 -0.0252>, rot=<-0.002 0.0067 0.2558 -0.9667>}
 middle_metacarpal_l {parent=hand_l, loc=<2.1856 -0.1725 -0.5984>, rot=<0.006 0.0168 -0.0329 -0.9993>}
 middle_01_l {parent=middle_metacarpal_l, loc=<7.3275 -0.2447 0.0129>, rot=<-0.0208 0.0042 0.0858 -0.9961>}
 middle_02_l {parent=middle_01_l, loc=<4.0123 0.1401 0.1175>, rot=<-0.0218 0.0258 0.3512 -0.9357>}
 middle_03_l {parent=middle_02_l, loc=<2.6077 0.0534 0.1373>, rot=<0.0082 0.0188 0.1708 -0.9851>}
 pinky_metacarpal_l {parent=hand_l, loc=<1.7616 -0.9753 3.3814>, rot=<-0.0397 0.0814 -0.069 -0.9935>}
 pinky_01_l {parent=pinky_metacarpal_l, loc=<7.1709 -0.4831 -0.1094>, rot=<-0.0654 -0.0035 0.2242 -0.9723>}
 pinky_02_l {parent=pinky_01_l, loc=<2.8274 0.0893 -0.1043>, rot=<0.0037 0.0011 0.2854 -0.9584>}
 pinky_03_l {parent=pinky_02_l, loc=<2.2046 0.0308 -0.0078>, rot=<-0.0002 0.0098 0.1445 -0.9895>}
 ring_metacarpal_l {parent=hand_l, loc=<2.2074 -0.3625 1.3638>, rot=<-0.0284 0.0722 -0.0458 -0.9959>}
 ring_01_l {parent=ring_metacarpal_l, loc=<7.0295 -0.6197 -0.1446>, rot=<-0.0489 -0.0252 0.1892 -0.9804>}
 ring_02_l {parent=ring_01_l, loc=<3.6174 0.4647 -0.1027>, rot=<0 0 0.2806 -0.9598>}
 ring_03_l {parent=ring_02_l, loc=<2.4152 0.1141 0.0199>, rot=<0 0 0.1189 -0.9929>}
 thumb_01_l {parent=hand_l, loc=<1.3118 -0.3191 -2.459>, rot=<0.6546 -0.4042 0.0664 -0.6354>}
 thumb_02_l {parent=thumb_01_l, loc=<5.9336 -0.0448 0.0401>, rot=<-0.0037 0.0213 0.298 -0.9543>}
 thumb_03_l {parent=thumb_02_l, loc=<2.2223 0.3682 -0.0907>, rot=<0.0118 -0.013 0.0066 -0.9998>}
 weapon_l {parent=hand_l, loc=<-0.1974 -3.4111 0.3811>, rot=<0 0 0 -1>}
 lowerarm_twist_01_l {parent=lowerarm_l, loc=<18.316 0 0>, rot=<0 0 0 1>}
 lowerarm_twist_02_l {parent=lowerarm_l, loc=<10.6707 0 0>, rot=<0 0 0 1>}
 upperarm_twist_01_l {parent=upperarm_l, loc=<7.2135 -0.1134 0>, rot=<0 0 0 1>}
 upperarm_twist_02_l {parent=upperarm_l, loc=<19.7998 -0.1134 0>, rot=<0 0 0 1>}
 clavicle_r {parent=spine_05, loc=<5.4831 -0.1053 3.2946>, rot=<0.7289 0.0465 -0.6829 0.009>}
 upperarm_r {parent=clavicle_r, loc=<-9.8658 1.2136 0.4104>, rot=<0.0164 -0.3245 -0.0269 -0.9454>}
 lowerarm_r {parent=upperarm_r, loc=<-26.4651 0.8345 -0.259>, rot=<0.0036 0.0161 -0.2298 -0.9731>}
 hand_r {parent=lowerarm_r, loc=<-22.7356 0.3203 0.6106>, rot=<-0.6784 -0.0223 0.0735 -0.7307>}
 index_metacarpal_r {parent=hand_r, loc=<-1.6534 -0.0181 2.9601>, rot=<0.0158 -0.0124 -0.0173 -0.9996>}
 index_01_r {parent=index_metacarpal_r, loc=<-8.3008 -0.0357 0.0524>, rot=<-0.0044 0.0485 0.0924 -0.9945>}
 index_02_r {parent=index_01_r, loc=<-3.3589 0.3003 -0.0453>, rot=<-0.002 0.0027 0.3225 -0.9466>}
 index_03_r {parent=index_02_r, loc=<-2.5578 0.123 0.0252>, rot=<-0.002 0.0067 0.2558 -0.9667>}
 middle_metacarpal_r {parent=hand_r, loc=<-2.1856 0.1726 0.5984>, rot=<0.006 0.0168 -0.0329 -0.9993>}
 middle_01_r {parent=middle_metacarpal_r, loc=<-7.3273 0.2447 -0.0129>, rot=<-0.0208 0.0042 0.0858 -0.9961>}
 middle_02_r {parent=middle_01_r, loc=<-4.0124 -0.14 -0.1175>, rot=<-0.0218 0.0258 0.3512 -0.9357>}
 middle_03_r {parent=middle_02_r, loc=<-2.6077 -0.0533 -0.1373>, rot=<0.0082 0.0188 0.1708 -0.9851>}
 pinky_metacarpal_r {parent=hand_r, loc=<-1.7617 0.9753 -3.3814>, rot=<-0.0397 0.0814 -0.069 -0.9935>}
 pinky_01_r {parent=pinky_metacarpal_r, loc=<-7.1707 0.4832 0.1094>, rot=<-0.0654 -0.0035 0.2242 -0.9723>}
 pinky_02_r {parent=pinky_01_r, loc=<-2.8274 -0.0892 0.1043>, rot=<0.0037 0.0011 0.2854 -0.9584>}
 pinky_03_r {parent=pinky_02_r, loc=<-2.2045 -0.031 0.0078>, rot=<-0.0002 0.0098 0.1445 -0.9895>}
 ring_metacarpal_r {parent=hand_r, loc=<-2.2073 0.3627 -1.3638>, rot=<-0.0284 0.0722 -0.0458 -0.9959>}
 ring_01_r {parent=ring_metacarpal_r, loc=<-7.0297 0.6197 0.1446>, rot=<-0.0489 -0.0252 0.1892 -0.9804>}
 ring_02_r {parent=ring_01_r, loc=<-3.6172 -0.4647 0.1027>, rot=<0 0 0.2806 -0.9598>}
 ring_03_r {parent=ring_02_r, loc=<-2.4152 -0.1141 -0.0199>, rot=<0 0 0.1189 -0.9929>}
 thumb_01_r {parent=hand_r, loc=<-1.3119 0.3192 2.459>, rot=<0.6546 -0.4042 0.0664 -0.6354>}
 thumb_02_r {parent=thumb_01_r, loc=<-5.9336 0.0448 -0.0401>, rot=<-0.0037 0.0213 0.298 -0.9543>}
 thumb_03_r {parent=thumb_02_r, loc=<-2.2223 -0.3682 0.0907>, rot=<0.0118 -0.013 0.0066 -0.9998>}
 weapon_r {parent=hand_r, loc=<0.1975 3.4111 -0.3811>, rot=<0 0 0 -1>}
 lowerarm_twist_01_r {parent=lowerarm_r, loc=<-18.3161 0 0>, rot=<0 0 0 1>}
 lowerarm_twist_02_r {parent=lowerarm_r, loc=<-10.6707 0 0>, rot=<0 0 0 1>}
 upperarm_twist_01_r {parent=upperarm_r, loc=<-7.2135 0.1134 0>, rot=<0 0 0 1>}
 upperarm_twist_02_r {parent=upperarm_r, loc=<-19.7998 0.1134 0>, rot=<0 0 0 1>}
 neck_01 {parent=spine_05, loc=<11.3138 0.1529 0>, rot=<0 0 -0.2003 -0.9797>}
 neck_02 {parent=neck_01, loc=<4.5627 0 0>, rot=<0 0 0.0504 -0.9987>}
 head {parent=neck_02, loc=<5.3914 0 0>, rot=<0 0 0.115 -0.9934>}
 dyn_neck {parent=neck_01, loc=<0 0 0>, rot=<0 0 0 1>}
 dyn_collar_l {parent=spine_05, loc=<10.3175 1.8909 -6.8779>, rot=<0.9788 0.1491 -0.0212 -0.139>}
 dyn_collar_back_l {parent=spine_05, loc=<12.119 -3.8369 -3.0652>, rot=<-0.8447 -0.1287 -0.0782 -0.5135>}
 dyn_collar_back_r {parent=spine_05, loc=<12.1195 -3.8368 3.0652>, rot=<-0.5482 -0.0835 -0.1253 -0.8227>}
 dyn_collar_r {parent=spine_05, loc=<10.3177 1.8909 6.8779>, rot=<0.0679 0.0103 -0.1503 -0.9863>}
 dyn_chest {parent=spine_05, loc=<4.835 6.5177 0>, rot=<0 0 -0.4993 -0.8665>}
 dyn_jacket_r {parent=spine_04, loc=<-3.8662 6.2147 10.7742>, rot=<0.0377 -0.9055 -0.4223 0.0176>}
 dyn_jacket_back_r {parent=spine_04, loc=<-1.5207 -6.7084 5.189>, rot=<0.0416 -0.9991 0 0>}
 dyn_jacket_back {parent=spine_04, loc=<-1.8986 -6.6416 -5.1889>, rot=<0.0416 -0.9991 0 0>}
 dyn_jacket_l {parent=spine_04, loc=<-3.4182 5.7617 -10.8516>, rot=<0.0377 -0.9055 -0.4223 0.0176>}
 thigh_l {parent=pelvis, loc=<-1.0855 -0.2109 -8.7053>, rot=<0.0009 0.0355 0.0155 -0.9993>}
 calf_l {parent=thigh_l, loc=<-42.5284 -0.0019 -0.7297>, rot=<0.0006 -0.0161 -0.0391 -0.9991>}
 calf_twist_01_l {parent=calf_l, loc=<-35.8483 -1.9197 -0.5794>, rot=<0.0237 0.0095 -0.0035 -0.9997>}
 calf_twist_02_l {parent=calf_l, loc=<-19.1546 -1.7943 -0.2638>, rot=<0.0237 0.0095 -0.0035 -0.9997>}
 foot_l {parent=calf_l, loc=<-46.7401 -2.5014 -0.592>, rot=<0.0427 -0.0238 0.0353 -0.9982>}
 ball_l {parent=foot_l, loc=<-3.5658 10.8066 -0.1544>, rot=<0.0001 -0.0001 -0.7063 -0.7079>}
 dyn_l_frontShoe {parent=foot_l, loc=<0 0 0>, rot=<0 0 0 -1>}
 dyn_l_backShoe {parent=foot_l, loc=<0 0 0>, rot=<0 0 0 -1>}
 thigh_twist_01_l {parent=thigh_l, loc=<-27.952 0 0>, rot=<0.0013 0 0 -1>}
 thigh_r {parent=pelvis, loc=<-1.0853 -0.2109 8.7053>, rot=<0.0355 -0.0009 0.9993 0.0155>}
 calf_r {parent=thigh_r, loc=<42.5286 0.0019 0.7297>, rot=<0.0006 -0.0161 -0.0391 -0.9991>}
 calf_twist_01_r {parent=calf_r, loc=<35.8485 1.9197 0.5794>, rot=<0.0237 0.0095 -0.0035 -0.9997>}
 calf_twist_02_r {parent=calf_r, loc=<19.1546 1.7943 0.2638>, rot=<0.0237 0.0095 -0.0035 -0.9997>}
 foot_r {parent=calf_r, loc=<46.7403 2.5014 0.592>, rot=<0.0427 -0.0238 0.0353 -0.9982>}
 ball_r {parent=foot_r, loc=<3.5658 -10.8066 0.1544>, rot=<0.0001 -0.0001 -0.7063 -0.7079>}
 dyn_r_backShoe {parent=foot_r, loc=<0 0 0>, rot=<0 0 0 -1>}
 dyn_r_frontShoe {parent=foot_r, loc=<0 0 0>, rot=<0 0 0 -1>}
 dyn_kneepad {parent=calf_r, loc=<0 0 0>, rot=<0 0 0 1>}
 thigh_twist_01_r {parent=thigh_r, loc=<27.952 0 0>, rot=<0.0013 0 0 -1>}
 dyn_l_butt {parent=pelvis, loc=<-1.0855 -0.2109 -8.7053>, rot=<0.0009 0.0355 0.0155 -0.9993>}
 dyn_strap_fr_l_MSTR {parent=pelvis, loc=<9.5218 6.2894 -6.3001>, rot=<0.0172 -0.8318 -0.541 -0.1227>}
 dyn_strap_fr_l_01 {parent=dyn_strap_fr_l_MSTR, loc=<0.5299 0 0>, rot=<0.0077 0.0279 0.0163 -0.9994>}
 dyn_strap_fr_l_02 {parent=dyn_strap_fr_l_01, loc=<7.4374 0 0>, rot=<0.0202 0.0197 0.0568 -0.998>}
 dyn_strap_fr_l_02a {parent=dyn_strap_fr_l_02, loc=<1.4195 0 0>, rot=<0.0784 0.0022 -0.0066 -0.9969>}
 dyn_strap_fr_l_03 {parent=dyn_strap_fr_l_02a, loc=<6.6999 0.0725 -0.3857>, rot=<-0.0436 -0.0941 0.0966 -0.9899>}
 dyn_strap_fr_l_04 {parent=dyn_strap_fr_l_03, loc=<6.9489 0.1736 -0.1699>, rot=<-0.0482 -0.1556 0.1367 -0.9771>}
 dyn_strap_fr_l_05 {parent=dyn_strap_fr_l_04, loc=<5.1951 0.1264 -0.0749>, rot=<0.0656 -0.2729 0.1001 -0.9546>}
 dyn_strap_fr_l_06 {parent=dyn_strap_fr_l_05, loc=<4.3319 -0.2897 -0.1614>, rot=<-0.198 -0.135 0.1092 -0.9647>}
 dyn_strap_bk_l_MSTR {parent=pelvis, loc=<10.8862 -6.9813 -6.1483>, rot=<0.0777 -0.8111 0.5221 -0.252>}
 dyn_strap_bk_l_01 {parent=dyn_strap_bk_l_MSTR, loc=<0.9205 0 0>, rot=<0.0123 -0.0154 0.0122 -0.9997>}
 dyn_strap_bk_l_02 {parent=dyn_strap_bk_l_01, loc=<8.8342 0 0>, rot=<0.0102 -0.126 -0.101 -0.9868>}
 dyn_strap_bk_l_03 {parent=dyn_strap_bk_l_02, loc=<7.8337 0 0>, rot=<0.0056 -0.1539 -0.121 -0.9806>}
 dyn_strap_bk_l_04 {parent=dyn_strap_bk_l_03, loc=<7.756 0 0>, rot=<0.0378 -0.1435 -0.1851 -0.9714>}
 dyn_strap_bk_l_05 {parent=dyn_strap_bk_l_04, loc=<6.8346 0 0>, rot=<0.0869 -0.2 -0.1692 -0.9611>}
 dyn_strap_bk_l_06 {parent=dyn_strap_bk_l_05, loc=<5.1591 0 0>, rot=<0.1018 -0.0719 -0.1067 -0.9864>}
 dyn_strap_fr_r_MSTR {parent=pelvis, loc=<9.5213 6.2894 6.3001>, rot=<0.123 0.5297 0.8391 -0.0155>}
 dyn_strap_fr_r_01 {parent=dyn_strap_fr_r_MSTR, loc=<0.5296 0 0>, rot=<0.0062 0.0285 -0.0151 -0.9995>}
 dyn_strap_fr_r_02 {parent=dyn_strap_fr_r_01, loc=<7.4372 0 0>, rot=<-0.0926 0.0187 -0.0571 -0.9939>}
 dyn_strap_fr_r_02a {parent=dyn_strap_fr_r_02, loc=<1.4194 0 0>, rot=<-0.0124 -0.0266 0.0103 -0.9995>}
 dyn_strap_fr_r_03 {parent=dyn_strap_fr_r_02a, loc=<6.7114 0 0>, rot=<0.0755 -0.0707 -0.0928 -0.9903>}
 dyn_strap_fr_r_04 {parent=dyn_strap_fr_r_03, loc=<6.9532 0 0>, rot=<0.0138 -0.1419 -0.1439 -0.9793>}
 dyn_strap_fr_r_05 {parent=dyn_strap_fr_r_04, loc=<5.1972 0 0>, rot=<0.2083 -0.2255 -0.2234 -0.9251>}
 dyn_strap_fr_r_06 {parent=dyn_strap_fr_r_05, loc=<4.3445 0 0>, rot=<0.0257 -0.2522 -0.1132 -0.9607>}
 dyn_strap_bk_r_MSTR {parent=pelvis, loc=<10.8858 -6.9813 6.1483>, rot=<0.2471 -0.4744 0.8399 -0.0921>}
 dyn_strap_bk_r_01 {parent=dyn_strap_bk_r_MSTR, loc=<0.9204 0 0>, rot=<0.0212 -0.0226 -0.0361 -0.9989>}
 dyn_strap_bk_r_02 {parent=dyn_strap_bk_r_01, loc=<8.8217 -0.4473 0.1005>, rot=<0.0527 -0.1238 0.1237 -0.9832>}
 dyn_strap_bk_r_03 {parent=dyn_strap_bk_r_02, loc=<7.834 0 0>, rot=<0.0176 -0.1664 0.1032 -0.9805>}
 dyn_strap_bk_r_04 {parent=dyn_strap_bk_r_03, loc=<7.756 0 0>, rot=<-0.0483 -0.1653 0.166 -0.971>}
 dyn_strap_bk_r_05 {parent=dyn_strap_bk_r_04, loc=<6.8346 0 0>, rot=<-0.0913 -0.2172 0.1465 -0.9607>}
 dyn_strap_bk_r_06 {parent=dyn_strap_bk_r_05, loc=<5.1591 0 0>, rot=<-0.1534 -0.1584 0.0878 -0.9714>}
 dyn_r_butt {parent=pelvis, loc=<-1.0853 -0.2109 8.7053>, rot=<0.0355 -0.0009 0.9993 0.0155>}
 ik_foot_root {parent=root, loc=<0 0 0>, rot=<0 0 0 1>}
 ik_foot_l {parent=ik_foot_root, loc=<14.8262 -7.3545 8.296>, rot=<-0.0357 -0.7045 -0.0284 0.7082>}
 ik_foot_r {parent=ik_foot_root, loc=<-14.8262 -7.3546 8.2958>, rot=<-0.7082 -0.0284 0.7045 -0.0357>}
 ik_hand_root {parent=root, loc=<0 0 0>, rot=<0 0 0 1>}
 ik_hand_gun {parent=ik_hand_root, loc=<-48.1744 3.3546 104.4381>, rot=<-0.6276 -0.1309 -0.4183 0.6434>}
 ik_hand_l {parent=ik_hand_gun, loc=<59.3272 67.6997 34.3524>, rot=<-0.6158 -0.7027 -0.3565 0>}
 ik_hand_r {parent=ik_hand_gun, loc=<0 0 0>, rot=<0 0 0 -1>}
]
