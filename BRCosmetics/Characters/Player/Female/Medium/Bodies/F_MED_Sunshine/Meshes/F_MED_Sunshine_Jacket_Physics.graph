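# Physics asset for the jacket of the F_MED_Sunshine female medium player cosmetic body, providing bodies for secondary motion on the jacket bones.
# Four thin box bodies cover the left, right and two back dyn_jacket panels, each constrained to spine_04 with limited swing and tight twist.
# Capsules along the pelvis and five spine bones form the torso's collision volume, chained by constraints with free swing and twist.
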PhysicsAsset /BRCosmetics/Characters/Player/Female/Medium/Bodies/F_MED_Sunshine/Meshes/F_MED_Sunshine_Jacket_Physics
bodies {
 dyn_jacket_back [{shapeType=Box, extentX=10.6047, extentY=0.5, extentZ=13.9148, center=<5.4011 0 0>, rotation=<rot 0 0 11.068>}]
 dyn_jacket_back_r [{shapeType=Box, extentX=10.6047, extentY=0.5, extentZ=13.9148, center=<5.4011 0 0>, rotation=<rot 0 0 -9.327>}]
 dyn_jacket_l [{shapeType=Box, extentX=10.6047, extentY=2.3127, extentZ=13.9148, center=<5.4011 0.2456 -0.5597>, rotation=<rot 0 0 72.0846>}]
 dyn_jacket_r [{shapeType=Box, extentX=10.6047, extentY=2.8368, extentZ=13.9148, center=<5.4011 -0.5378 0.1122>, rotation=<rot 0 0 11.7856>}]
 pelvis [{shapeType=Capsule, length=2, radius=2}]
 spine_01 [{shapeType=Capsule, length=6.8247, radius=7.9835}]
 spine_02 [{shapeType=Capsule, length=9.4875, radius=5.8958}]
 spine_03 [{shapeType=Capsule, length=7.9012, radius=6.2495}]
 spine_04 [{shapeType=Capsule, length=6.0343, radius=7.0109}]
 spine_05 [{shapeType=Capsule, length=11.7448, radius=7.5709}]
}
constraints [
 {bone1Name=dyn_jacket_r, bone2Name=spine_04, swing1LimitDegrees=45, swing2LimitDegrees=10, twistLimitDegrees=2}
 {bone1Name=dyn_jacket_back_r, bone2Name=spine_04, swing1LimitDegrees=45, swing2LimitDegrees=10, twistLimitDegrees=2}
 {bone1Name=dyn_jacket_back, bone2Name=spine_04, swing1LimitDegrees=45, swing2LimitDegrees=10, twistLimitDegrees=2}
 {bone1Name=dyn_jacket_l, bone2Name=spine_04, swing1LimitDegrees=10, swing2LimitDegrees=45, twistLimitDegrees=2}
 {
  bone1Name=spine_02
  bone2Name=spine_01
  swing1LimitDegrees=45
  swing1Motion=Free
  swing2LimitDegrees=45
  swing2Motion=Free
  twistLimitDegrees=45
  twistMotion=Free
 }
 {
  bone1Name=spine_01
  bone2Name=pelvis
  swing1LimitDegrees=45
  swing1Motion=Free
  swing2LimitDegrees=45
  swing2Motion=Free
  twistLimitDegrees=45
  twistMotion=Free
 }
 {
  bone1Name=spine_05
  bone2Name=spine_04
  swing1LimitDegrees=45
  swing1Motion=Free
  swing2LimitDegrees=45
  swing2Motion=Free
  twistLimitDegrees=45
  twistMotion=Free
 }
 {
  bone1Name=spine_04
  bone2Name=spine_03
  swing1LimitDegrees=45
  swing1Motion=Free
  swing2LimitDegrees=45
  swing2Motion=Free
  twistLimitDegrees=45
  twistMotion=Free
 }
 {
  bone1Name=spine_03
  bone2Name=spine_02
  swing1LimitDegrees=45
  swing1Motion=Free
  swing2LimitDegrees=45
  swing2Motion=Free
  twistLimitDegrees=45
  twistMotion=Free
 }
]
