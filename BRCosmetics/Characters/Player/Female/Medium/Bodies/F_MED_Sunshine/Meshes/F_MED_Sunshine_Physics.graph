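# Physics asset for the F_MED_Sunshine female player body, covering four dynamic strap bone chains (front/back, left/right).
# Capsules form each strap link and cover the pelvis, lower spine and thighs.
# Each chain hangs from a free-rotating master body on the pelvis, with angle-limited joints between links.
# Extra free-rotation constraints link front and back strap ends on each side, and thighs to the first front links.
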PhysicsAsset /BRCosmetics/Characters/Player/Female/Medium/Bodies/F_MED_Sunshine/Meshes/F_MED_Sunshine_Physics
bodies {
 dyn_strap_bk_l_01 [{shapeType=Capsule, length=3.615, radius=5, shapeName=dyn_strap_fr_l_05_capsule, center=<3.4279 -0.1263 4>, rotation=<rot 0 86.6218 -87.0721>}]
 dyn_strap_bk_l_02 [{
   shapeType=Capsule
   length=5.7112
   radius=5
   shapeName=dyn_strap_fr_l_05_capsule
   center=<3.3517 -0.122 4.5>
   rotation=<rot 0 86.6218 -87.0721>
  }]
 dyn_strap_bk_l_03 [{shapeType=Capsule, length=6.75, radius=5, shapeName=dyn_strap_fr_l_05_capsule, center=<1.2871 0 4.5>, rotation=<rot 0 86.6218 -87.0721>}]
 dyn_strap_bk_l_04 [{shapeType=Capsule, length=6.75, radius=5, shapeName=dyn_strap_fr_l_05_capsule, center=<1.2871 0 4>, rotation=<rot 0 86.6218 -87.0721>}]
 dyn_strap_bk_l_05 [{shapeType=Capsule, length=6.3773, radius=5, shapeName=dyn_strap_fr_l_05_capsule, center=<1.2871 0 4>, rotation=<rot 0 86.6218 -87.0721>}]
 dyn_strap_bk_l_06 [{shapeType=Capsule, length=6.1997, radius=5, shapeName=dyn_strap_fr_l_05_capsule, center=<1.2871 0 4>, rotation=<rot 0 86.6218 -87.0721>}]
 dyn_strap_bk_l_MSTR [{shapeType=Capsule, length=1, radius=1, shapeName=dyn_strap_bk_l_MSTR_capsule}]
 dyn_strap_bk_r_01 [{shapeType=Capsule, length=3.615, radius=5, shapeName=dyn_strap_fr_l_05_capsule, center=<3.4279 -0.1263 4>, rotation=<rot 0 86.6218 -87.0721>}]
 dyn_strap_bk_r_02 [{shapeType=Capsule, length=5.7112, radius=5, shapeName=dyn_strap_fr_l_05_capsule, center=<3.3517 -0.122 4>, rotation=<rot 0 86.6218 -87.0721>}]
 dyn_strap_bk_r_03 [{shapeType=Capsule, length=6.75, radius=5, shapeName=dyn_strap_fr_l_05_capsule, center=<1.2871 0 4>, rotation=<rot 0 86.6218 -87.0721>}]
 dyn_strap_bk_r_04 [{shapeType=Capsule, length=6.75, radius=5, shapeName=dyn_strap_fr_l_05_capsule, center=<1.2871 0 4>, rotation=<rot 0 86.6218 -87.0721>}]
 dyn_strap_bk_r_05 [{shapeType=Capsule, length=5.9286, radius=5, shapeName=dyn_strap_fr_l_05_capsule, center=<1.2871 0 4>, rotation=<rot 0 86.6218 -87.0721>}]
 dyn_strap_bk_r_06 [{shapeType=Capsule, length=5.9286, radius=5, shapeName=dyn_strap_fr_l_05_capsule, center=<1.2871 0 4>, rotation=<rot 0 86.6218 -87.0721>}]
 dyn_strap_bk_r_MSTR [{shapeType=Capsule, length=1, radius=1, shapeName=dyn_strap_bk_r_MSTR_capsule}]
 dyn_strap_fr_l_01 [{shapeType=Capsule, length=4.25, radius=5, center=<2.6599 0 4>, rotation=<rot 90 0 0>}]
 dyn_strap_fr_l_02 [{shapeType=Capsule, length=6.5131, radius=5, shapeName=dyn_strap_fr_l_05_capsule, center=<1.2871 0 4>, rotation=<rot 0 90 90>}]
 dyn_strap_fr_l_02a [{shapeType=Capsule, length=6.7631, radius=5, shapeName=dyn_strap_fr_l_05_capsule, center=<1.2871 0 4>, rotation=<rot 0 90 90>}]
 dyn_strap_fr_l_03 [{shapeType=Capsule, length=9.2631, radius=5, shapeName=dyn_strap_fr_l_05_capsule, center=<1.2871 0 4>, rotation=<rot 0 90 90>}]
 dyn_strap_fr_l_04 [{shapeType=Capsule, length=9.2631, radius=5, shapeName=dyn_strap_fr_l_05_capsule, center=<1.2871 0 4>, rotation=<rot 0 90 90>}]
 dyn_strap_fr_l_05 [{shapeType=Capsule, length=8.6381, radius=5, shapeName=dyn_strap_fr_l_05_capsule, center=<1.2871 0 4>, rotation=<rot 0 90 -90>}]
 dyn_strap_fr_l_06 [{shapeType=Capsule, length=6.5131, radius=5, shapeName=dyn_strap_fr_l_05_capsule, center=<1.2871 0 4>, rotation=<rot 0 90 -90>}]
 dyn_strap_fr_l_MSTR [{shapeType=Capsule, length=1, radius=1}]
 dyn_strap_fr_r_01 [{shapeType=Capsule, length=4.25, radius=5, center=<2.6599 0 4>, rotation=<rot 90 0 0>}]
 dyn_strap_fr_r_02 [{shapeType=Capsule, length=6.7631, radius=5, shapeName=dyn_strap_fr_l_05_capsule, center=<1.2871 0 4>, rotation=<rot 0 86.6218 -87.0721>}]
 dyn_strap_fr_r_02a [{shapeType=Capsule, length=6.7631, radius=5, shapeName=dyn_strap_fr_l_05_capsule, center=<1.2871 0 4>, rotation=<rot 0 86.6218 -87.0721>}]
 dyn_strap_fr_r_03 [{shapeType=Capsule, length=8.8881, radius=5, shapeName=dyn_strap_fr_l_05_capsule, center=<1.2871 0 4>, rotation=<rot 0 86.6218 -87.0721>}]
 dyn_strap_fr_r_04 [{shapeType=Capsule, length=8.8881, radius=5, shapeName=dyn_strap_fr_l_05_capsule, center=<1.2871 0 4>, rotation=<rot 0 86.6218 -87.0721>}]
 dyn_strap_fr_r_05 [{shapeType=Capsule, length=5.9286, radius=5, shapeName=dyn_strap_fr_l_05_capsule, center=<1.2871 0 4>, rotation=<rot 0 86.6218 -87.0721>}]
 dyn_strap_fr_r_06 [{shapeType=Capsule, length=5.9286, radius=5, shapeName=dyn_strap_fr_l_05_capsule, center=<1.2871 0 4>, rotation=<rot 0 86.6218 -87.0721>}]
 dyn_strap_fr_r_MSTR [{shapeType=Capsule, length=1, radius=1}]
 pelvis [
  {shapeType=Capsule, length=9.7487, radius=9.7826, center=<-2.5033 -2.528 -0.0002>}
  {shapeType=Capsule, length=33.3657, radius=10.8413, center=<-17.5896 -1.0202 0.0004>, rotation=<rot 90 0.0198 0.0199>}
  {shapeType=Capsule, length=15, radius=6.1259, center=<4.538 -2.2413 0>}
  {
   shapeType=Capsule
   length=37.2958
   radius=6.9111
   shapeName=pelvis_capsule
   center=<-17.6333 8.7563 0.0002>
   rotation=<rot 90 -0.0395 -0.0394>
  }
 ]
 spine_01 [{shapeType=Capsule, length=7.1188, radius=7.3821, center=<4.538 0.6326 -0.0001>}]
 spine_02 [{shapeType=Capsule, length=6.125, radius=8.3759, center=<4.538 0.6326 -0.0001>}]
 thigh_l [{shapeType=Capsule, length=33.3863, radius=8.4926, center=<-21.1166 -0.8915 0.0108>, rotation=<rot 90 0 0>}]
 thigh_r [{shapeType=Capsule, length=33.4245, radius=8.3567, center=<21.1166 0.9564 0.0528>, rotation=<rot 90 0 0>}]
}
constraints [
 {
  bone1Name=thigh_l
  bone2Name=pelvis
  swing1LimitDegrees=45
  swing1Motion=Free
  swing2LimitDegrees=45
  swing2Motion=Free
  twistLimitDegrees=45
  twistMotion=Free
 }
 {
  bone1Name=thigh_r
  bone2Name=pelvis
  swing1LimitDegrees=45
  swing1Motion=Free
  swing2LimitDegrees=45
  swing2Motion=Free
  twistLimitDegrees=45
  twistMotion=Free
 }
 {
  bone1Name=spine_01
  bone2Name=pelvis
  swing1LimitDegrees=45
  swing1Motion=Free
  swing2LimitDegrees=45
  swing2Motion=Free
  twistLimitDegrees=45
  twistMotion=Free
 }
 {
  bone1Name=spine_02
  bone2Name=spine_01
  swing1LimitDegrees=45
  swing1Motion=Free
  swing2LimitDegrees=45
  swing2Motion=Free
  twistLimitDegrees=45
  twistMotion=Free
 }
 {
  bone1Name=dyn_strap_fr_l_MSTR
  bone2Name=pelvis
  swing1LimitDegrees=45
  swing1Motion=Free
  swing2LimitDegrees=45
  swing2Motion=Free
  twistLimitDegrees=45
  twistMotion=Free
 }
 {bone1Name=dyn_strap_fr_l_01, bone2Name=dyn_strap_fr_l_MSTR, swing1LimitDegrees=45, swing2LimitDegrees=45, twistLimitDegrees=25}
 {bone1Name=dyn_strap_fr_l_02, bone2Name=dyn_strap_fr_l_01, swing1LimitDegrees=45, swing2LimitDegrees=45, twistLimitDegrees=25}
 {bone1Name=dyn_strap_fr_l_02a, bone2Name=dyn_strap_fr_l_02, swing1LimitDegrees=45, swing2LimitDegrees=45, twistLimitDegrees=25}
 {bone1Name=dyn_strap_fr_l_03, bone2Name=dyn_strap_fr_l_02a, swing1LimitDegrees=45, swing2LimitDegrees=45, twistLimitDegrees=25}
 {bone1Name=dyn_strap_fr_l_04, bone2Name=dyn_strap_fr_l_03, swing1LimitDegrees=45, swing2LimitDegrees=45, twistLimitDegrees=25}
 {bone1Name=dyn_strap_fr_l_05, bone2Name=dyn_strap_fr_l_04, swing1LimitDegrees=45, swing2LimitDegrees=45, twistLimitDegrees=25}
 {bone1Name=dyn_strap_fr_l_06, bone2Name=dyn_strap_fr_l_05, swing1LimitDegrees=45, swing2LimitDegrees=45, twistLimitDegrees=25}
 {
  bone1Name=dyn_strap_bk_l_MSTR
  bone2Name=pelvis
  swing1LimitDegrees=45
  swing1Motion=Free
  swing2LimitDegrees=45
  swing2Motion=Free
  twistLimitDegrees=45
  twistMotion=Free
 }
 {bone1Name=dyn_strap_bk_l_01, bone2Name=dyn_strap_bk_l_MSTR, swing1LimitDegrees=45, swing2LimitDegrees=45, twistLimitDegrees=25}
 {bone1Name=dyn_strap_bk_l_02, bone2Name=dyn_strap_bk_l_01, swing1LimitDegrees=45, swing2LimitDegrees=45, twistLimitDegrees=25}
 {bone1Name=dyn_strap_bk_l_03, bone2Name=dyn_strap_bk_l_02, swing1LimitDegrees=45, swing2LimitDegrees=45, twistLimitDegrees=25}
 {bone1Name=dyn_strap_bk_l_04, bone2Name=dyn_strap_bk_l_03, swing1LimitDegrees=45, swing2LimitDegrees=45, twistLimitDegrees=25}
 {bone1Name=dyn_strap_bk_l_05, bone2Name=dyn_strap_bk_l_04, swing1LimitDegrees=45, swing2LimitDegrees=45, twistLimitDegrees=25}
 {bone1Name=dyn_strap_bk_l_06, bone2Name=dyn_strap_bk_l_05, swing1LimitDegrees=45, swing2LimitDegrees=45, twistLimitDegrees=25}
 {
  bone1Name=dyn_strap_fr_r_MSTR
  bone2Name=pelvis
  swing1LimitDegrees=45
  swing1Motion=Free
  swing2LimitDegrees=45
  swing2Motion=Free
  twistLimitDegrees=45
  twistMotion=Free
 }
 {bone1Name=dyn_strap_fr_r_01, bone2Name=dyn_strap_fr_r_MSTR, swing1LimitDegrees=45, swing2LimitDegrees=45, twistLimitDegrees=25}
 {bone1Name=dyn_strap_fr_r_02, bone2Name=dyn_strap_fr_r_01, swing1LimitDegrees=45, swing2LimitDegrees=45, twistLimitDegrees=25}
 {bone1Name=dyn_strap_fr_r_02a, bone2Name=dyn_strap_fr_r_02, swing1LimitDegrees=45, swing2LimitDegrees=45, twistLimitDegrees=25}
 {bone1Name=dyn_strap_fr_r_03, bone2Name=dyn_strap_fr_r_02a, swing1LimitDegrees=45, swing2LimitDegrees=45, twistLimitDegrees=25}
 {bone1Name=dyn_strap_fr_r_04, bone2Name=dyn_strap_fr_r_03, swing1LimitDegrees=45, swing2LimitDegrees=45, twistLimitDegrees=25}
 {bone1Name=dyn_strap_fr_r_05, bone2Name=dyn_strap_fr_r_04, swing1LimitDegrees=45, swing2LimitDegrees=45, twistLimitDegrees=25}
 {bone1Name=dyn_strap_fr_r_06, bone2Name=dyn_strap_fr_r_05, swing1LimitDegrees=45, swing2LimitDegrees=45, twistLimitDegrees=25}
 {
  bone1Name=dyn_strap_bk_r_MSTR
  bone2Name=pelvis
  swing1LimitDegrees=45
  swing1Motion=Free
  swing2LimitDegrees=45
  swing2Motion=Free
  twistLimitDegrees=45
  twistMotion=Free
 }
 {bone1Name=dyn_strap_bk_r_01, bone2Name=dyn_strap_bk_r_MSTR, swing1LimitDegrees=45, swing2LimitDegrees=45, twistLimitDegrees=25}
 {bone1Name=dyn_strap_bk_r_02, bone2Name=dyn_strap_bk_r_01, swing1LimitDegrees=45, swing2LimitDegrees=45, twistLimitDegrees=25}
 {bone1Name=dyn_strap_bk_r_03, bone2Name=dyn_strap_bk_r_02, swing1LimitDegrees=45, swing2LimitDegrees=45, twistLimitDegrees=25}
 {bone1Name=dyn_strap_bk_r_04, bone2Name=dyn_strap_bk_r_03, swing1LimitDegrees=45, swing2LimitDegrees=45, twistLimitDegrees=25}
 {bone1Name=dyn_strap_bk_r_05, bone2Name=dyn_strap_bk_r_04, swing1LimitDegrees=45, swing2LimitDegrees=45, twistLimitDegrees=25}
 {bone1Name=dyn_strap_bk_r_06, bone2Name=dyn_strap_bk_r_05, swing1LimitDegrees=45, swing2LimitDegrees=45, twistLimitDegrees=25}
 {
  bone1Name=dyn_strap_fr_l_05
  bone2Name=dyn_strap_bk_l_06
  swing1LimitDegrees=45
  swing1Motion=Free
  swing2LimitDegrees=45
  swing2Motion=Free
  twistLimitDegrees=45
  twistMotion=Free
 }
 {
  bone1Name=dyn_strap_fr_l_06
  bone2Name=dyn_strap_bk_l_06
  swing1LimitDegrees=45
  swing1Motion=Free
  swing2LimitDegrees=45
  swing2Motion=Free
  twistLimitDegrees=45
  twistMotion=Free
 }
 {
  bone1Name=dyn_strap_fr_r_06
  bone2Name=dyn_strap_bk_r_06
  swing1LimitDegrees=45
  swing1Motion=Free
  swing2LimitDegrees=45
  swing2Motion=Free
  twistLimitDegrees=45
  twistMotion=Free
 }
 {
  bone1Name=thigh_l
  bone2Name=dyn_strap_fr_l_01
  swing1LimitDegrees=45
  swing1Motion=Free
  swing2LimitDegrees=45
  swing2Motion=Free
  twistLimitDegrees=45
  twistMotion=Free
 }
 {
  bone1Name=thigh_r
  bone2Name=dyn_strap_fr_r_01
  swing1LimitDegrees=45
  swing1Motion=Free
  swing2LimitDegrees=45
  swing2Motion=Free
  twistLimitDegrees=45
  twistMotion=Free
 }
]
