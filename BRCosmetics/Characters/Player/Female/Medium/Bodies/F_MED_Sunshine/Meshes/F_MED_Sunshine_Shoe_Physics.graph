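# Physics asset for the shoes of the F_MED_Sunshine female medium player body, setting up physical motion for dynamic shoe bones.
# Capsule bodies cover both feet plus a front and back shoe bone on each side.
# Each shoe bone is constrained to its foot with narrow swing and tighter twist limits, so the pieces can move only slightly relative to the foot.
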PhysicsAsset /BRCosmetics/Characters/Player/Female/Medium/Bodies/F_MED_Sunshine/Meshes/F_MED_Sunshine_Shoe_Physics
bodies {
 dyn_l_backShoe [{shapeType=Capsule, length=3.2303, radius=2.8502, center=<-1.4022 0.0014 -0.0159>, rotation=<rot 90 0 0>}]
 dyn_l_frontShoe [{shapeType=Capsule, length=4.4672, radius=2.6076, center=<-1.4022 0.0014 -0.0159>, rotation=<rot 90 0 0>}]
 dyn_r_backShoe [{shapeType=Capsule, length=4.4672, radius=2.6076, center=<1.4023 0 0>, rotation=<rot 90 0 0>}]
 dyn_r_frontShoe [{shapeType=Capsule, length=3.2303, radius=2.8502, center=<1.4023 0 0>, rotation=<rot 90 0 0>}]
 foot_l [{shapeType=Capsule, length=5.0408, radius=3.603, rotation=<rot 90 0 0>}]
 foot_r [{shapeType=Capsule, length=5.0408, radius=3.603, rotation=<rot 90 0 0>}]
}
constraints [
 {bone1Name=dyn_l_frontShoe, bone2Name=foot_l, swing1LimitDegrees=15, swing2LimitDegrees=15, twistLimitDegrees=5}
 {bone1Name=dyn_l_backShoe, bone2Name=foot_l, swing1LimitDegrees=15, swing2LimitDegrees=15, twistLimitDegrees=5}
 {bone1Name=dyn_r_backShoe, bone2Name=foot_r, swing1LimitDegrees=15, swing2LimitDegrees=15, twistLimitDegrees=5}
 {bone1Name=dyn_r_frontShoe, bone2Name=foot_r, swing1LimitDegrees=15, swing2LimitDegrees=15, twistLimitDegrees=5}
]
