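# Secondary-motion AnimBP for the F_MED_Sunshine female cosmetic body, copying the attached parent's pose.
# Bone constraints, offsets, RigidBody and AnimDynamics drive straps, jacket, collar, kneepad and shoes, before a Control Rig weighted by crouch state.
# The event graph derives strap alphas from thigh rotation and air/crouch state, updates gravity override and resets dynamics on crouch changes.
# Jacket and shoe RigidBody nodes are disabled; Crouching is set true on both branches.
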AnimBlueprint /BRCosmetics/Characters/Player/Female/Medium/Bodies/F_MED_Sunshine/Meshes/F_MED_Sunshine_AnimBP
target_skeleton=&/F_MED_Sunshine_Skeleton
graphs [
 AnimGraph {nodes [
   n1 = "Constraint - dyn_strap_bk_l_MSTR (PPPPPPPPP)"!(
    ComponentPose=Local To Component(LocalPose=Copy Pose From Mesh() {bCopyCurves=true, bUseAttachedParent=true})
    Alpha=1
    ConstraintWeights_0=0.2442
    ConstraintWeights_1:float=0.1453
    ConstraintWeights_2:float=0.1137
    ConstraintWeights_3:float=0.1401
    ConstraintWeights_4:float=0.3446
    ConstraintWeights_5:float=0.0091
    ConstraintWeights_6:float=0.0026
    ConstraintWeights_7:float=0.0001
    ConstraintWeights_8:float=0.0001
   ) -> (Pose) {
    alphaBoolBlend {blendOption=Linear}
    boneToModify {boneName=dyn_strap_bk_l_MSTR}
    constraintSetup [
     {targetBone {boneName=spine_02}}
     {targetBone {boneName=spine_04}}
     {targetBone {boneName=spine_03}}
     {targetBone {boneName=spine_01}}
     {targetBone {boneName=pelvis}}
     {targetBone {boneName=thigh_twist_01_l}}
     {targetBone {boneName=thigh_twist_01_r}}
     {targetBone {boneName=thigh_r}}
     {targetBone {boneName=thigh_l}}
    ]
    constraintWeights [1,1,1,1,1,1,1,1,1]
   }
   n2 = "Constraint - dyn_strap_bk_r_MSTR (PPPPPPPPP)"!(
    ComponentPose=@n1
    Alpha=1
    ConstraintWeights_0=0.2442
    ConstraintWeights_1:float=0.1453
    ConstraintWeights_2:float=0.1137
    ConstraintWeights_3:float=0.1401
    ConstraintWeights_4:float=0.3446
    ConstraintWeights_5:float=0.0091
    ConstraintWeights_6:float=0.0026
    ConstraintWeights_7:float=0.0001
    ConstraintWeights_8:float=0.0001
   ) -> (Pose) {
    alphaBoolBlend {blendOption=Linear}
    boneToModify {boneName=dyn_strap_bk_r_MSTR}
    constraintSetup [
     {targetBone {boneName=spine_02}}
     {targetBone {boneName=spine_04}}
     {targetBone {boneName=spine_03}}
     {targetBone {boneName=spine_01}}
     {targetBone {boneName=pelvis}}
     {targetBone {boneName=thigh_twist_01_l}}
     {targetBone {boneName=thigh_twist_01_r}}
     {targetBone {boneName=thigh_r}}
     {targetBone {boneName=thigh_l}}
    ]
    constraintWeights [1,1,1,1,1,1,1,1,1]
   }
   n3 = "Constraint - dyn_strap_fr_l_MSTR (PPPPPPPPP)"!(
    ComponentPose=@n2
    Alpha=1
    ConstraintWeights_0=0.2442
    ConstraintWeights_1:float=0.1453
    ConstraintWeights_2:float=0.1137
    ConstraintWeights_3:float=0.1401
    ConstraintWeights_4:float=0.3446
    ConstraintWeights_5:float=0.0091
    ConstraintWeights_6:float=0.0026
    ConstraintWeights_7:float=0.0001
    ConstraintWeights_8:float=0.0001
   ) -> (Pose) {
    alphaBoolBlend {blendOption=Linear}
    boneToModify {boneName=dyn_strap_fr_l_MSTR}
    constraintSetup [
     {targetBone {boneName=spine_02}}
     {targetBone {boneName=spine_04}}
     {targetBone {boneName=spine_03}}
     {targetBone {boneName=spine_01}}
     {targetBone {boneName=pelvis}}
     {targetBone {boneName=thigh_twist_01_l}}
     {targetBone {boneName=thigh_twist_01_r}}
     {targetBone {boneName=thigh_r}}
     {targetBone {boneName=thigh_l}}
    ]
    constraintWeights [1,1,1,1,1,1,1,1,1]
   }
   n4 = "Constraint - dyn_strap_fr_r_MSTR (PPPPPPPPP)"!(
    ComponentPose=@n3
    Alpha=1
    ConstraintWeights_0=0.2442
    ConstraintWeights_1:float=0.1453
    ConstraintWeights_2:float=0.1137
    ConstraintWeights_3:float=0.1401
    ConstraintWeights_4:float=0.3446
    ConstraintWeights_5:float=0.0091
    ConstraintWeights_6:float=0.0026
    ConstraintWeights_7:float=0.0001
    ConstraintWeights_8:float=0.0001
   ) -> (Pose) {
    alphaBoolBlend {blendOption=Linear}
    boneToModify {boneName=dyn_strap_fr_r_MSTR}
    constraintSetup [
     {targetBone {boneName=spine_02}}
     {targetBone {boneName=spine_04}}
     {targetBone {boneName=spine_03}}
     {targetBone {boneName=spine_01}}
     {targetBone {boneName=pelvis}}
     {targetBone {boneName=thigh_twist_01_l}}
     {targetBone {boneName=thigh_twist_01_r}}
     {targetBone {boneName=thigh_r}}
     {targetBone {boneName=thigh_l}}
    ]
    constraintWeights [1,1,1,1,1,1,1,1,1]
   }
   n5 = Get Crouching() -> (Crouching:bool)
   n6 = "Constraint - dyn_strap_fr_l_01 (R)"!(ComponentPose=@n4, bAlphaBoolEnabled:bool=@n5, ConstraintWeights_0=1) -> (Pose) {
    alphaInputType=Bool
    alphaBoolBlend {blendInTime=10, blendOption=Linear, blendOutTime=10}
    boneToModify {boneName=dyn_strap_fr_l_01}
    constraintSetup [{transformType=Rotation, targetBone {boneName=thigh_l}}]
    constraintWeights [1]
   }
   n7 = "Transform (Modify) Bone - Bone: dyn_strap_fr_l_01"!(ComponentPose=@n6, bAlphaBoolEnabled:bool=@n5, Translation="0,0,5", Rotation="0,-55,0", Scale="1,1,1") -> (Pose) {
    alphaInputType=Bool
    rotationMode=BMM_Additive
    scaleMode=BMM_Additive
    translationMode=BMM_Additive
    alphaBoolBlend {blendOption=Linear}
    boneToModify {boneName=dyn_strap_fr_l_01}
    rotation {yaw=-55}
    translation {z=5}
   }
   n8 = "Transform (Modify) Bone - Bone: dyn_strap_fr_l_MSTR"(ComponentPose=@n7, Alpha=Get thigh_rot_l(), Translation="0.000,-0.517,2.019", Rotation="0,0.000,-20", Scale="1,1,1") {
    rotationMode=BMM_Additive
    translationMode=BMM_Additive
    alphaBoolBlend {blendOption=Linear}
    alphaScaleBiasClamp {bInterpResult=true, interpSpeedDecreasing=15, interpSpeedIncreasing=15}
    boneToModify {boneName=dyn_strap_fr_l_MSTR}
    rotation {roll=-20}
    translation {y=-0.517, z=2.019}
   }
   n9 = "Transform (Modify) Bone - Bone: dyn_strap_fr_r_MSTR"(ComponentPose=@n8, Alpha=Get thigh_rot_r(), Translation="0.674,-0.985,3.016", Rotation="0.000,0.000,20", Scale="1,1,1") {
    rotationMode=BMM_Additive
    translationMode=BMM_Additive
    alphaBoolBlend {blendOption=Linear}
    alphaScaleBiasClamp {bInterpResult=true, interpSpeedDecreasing=15, interpSpeedIncreasing=15}
    boneToModify {boneName=dyn_strap_fr_r_MSTR}
    rotation {roll=20}
    translation=<0.674 -0.985 3.016>
   }
   n10 = "Transform (Modify) Bone - Bone: dyn_strap_fr_r_MSTR"(ComponentPose=@n9, Alpha=Get Thigh_R_Rot_Crouch() -> (Thigh_R_Rot_Crouch:double), Translation="0,0,0", Rotation="0,90,45", Scale="1,1,1") {
    rotationMode=BMM_Additive
    rotationSpace=BCS_BoneSpace
    alphaBoolBlend {blendOption=Linear}
    boneToModify {boneName=dyn_strap_fr_r_MSTR}
    rotation {roll=45, yaw=90}
   }
   n11 = "Transform (Modify) Bone - Bone: dyn_strap_fr_l_MSTR"(ComponentPose=@n10, Alpha=Get Thigh L Rot Crouch() -> (Thigh L Rot Crouch:double), Translation="0,0,0", Rotation="0,-90,-45", Scale="1,1,1") {
    rotationMode=BMM_Additive
    rotationSpace=BCS_BoneSpace
    alphaBoolBlend {blendOption=Linear}
    boneToModify {boneName=dyn_strap_fr_l_MSTR}
    rotation {roll=-45, yaw=-90}
   }
   n12 = RigidBody!(ComponentPose=@n11, Alpha, OverrideWorldGravity:Vector=Get GravityOverride(), ExternalForce) -> (Pose) {
    bClampLinearTranslationLimitToRefPose=true
    bOverrideWorldGravity=true
    overridePhysicsAsset=&/F_MED_Sunshine_Physics
    simulationSpace=BaseBoneSpace
    alphaBoolBlend {blendOption=Linear}
    baseBoneRef {boneName=pelvis}
    componentAppliedLinearAccClamp=<3000 3000 3000>
    componentLinearAccScale=<0.2 0.2 0.2>
    componentLinearVelScale=<0.2 0.1 0.1>
    overrideWorldGravity {z=-980}
   }
   n13 = Anim Dynamics - Bone: dyn_jacket_r(ComponentPose=@n12, Alpha=0.4, ExternalForce="0,0,0") {
    angularDampingOverride=1
    angularSpringConstant=75
    bAngularSpring=true
    bGravityOverrideInSimSpace=true
    bOverrideAngularDamping=true
    simulationSpace=RootRelative
    alphaBoolBlend {blendOption=Linear}
    boundBone {boneName=dyn_jacket_r}
    componentAppliedLinearAccClamp=<5000 5000 5000>
    componentLinearAccScale=<1 1 1>
    componentLinearVelScale=<1 1 1>
    physicsBodyDefinitions [{
      sphereCollisionRadius=0
      boundBone {boneName=dyn_jacket_r}
      boxExtents=<20 10 10>
      constraintSetup {angularLimitsMax=<25 0 25>, angularLimitsMin=<-25 -25 -25>, angularTarget=<0 0 0>}
      localJointOffset=<2 0 0>
     }]
   }
   n14 = Anim Dynamics - Bone: dyn_jacket_back_r(ComponentPose=@n13, Alpha=0.4, ExternalForce="0,0,0") {
    angularDampingOverride=1
    angularSpringConstant=75
    bAngularSpring=true
    bGravityOverrideInSimSpace=true
    bOverrideAngularDamping=true
    simulationSpace=RootRelative
    alphaBoolBlend {blendOption=Linear}
    boundBone {boneName=dyn_jacket_back_r}
    componentAppliedLinearAccClamp=<5000 5000 5000>
    componentLinearAccScale=<1 1 1>
    componentLinearVelScale=<1 1 1>
    physicsBodyDefinitions [{
      sphereCollisionRadius=0
      boundBone {boneName=dyn_jacket_back_r}
      boxExtents=<20 10 10>
      constraintSetup {angularLimitsMax=<25 0 25>, angularLimitsMin=<-25 -25 -25>, angularTarget=<0 0 0>}
      localJointOffset=<2 0 0>
     }]
   }
   n15 = Anim Dynamics - Bone: dyn_jacket_back(ComponentPose=@n14, Alpha=0.4, ExternalForce="0,0,0") {
    angularDampingOverride=1
    angularSpringConstant=75
    bAngularSpring=true
    bGravityOverrideInSimSpace=true
    bOverrideAngularDamping=true
    simulationSpace=RootRelative
    alphaBoolBlend {blendOption=Linear}
    boundBone {boneName=dyn_jacket_back}
    componentAppliedLinearAccClamp=<5000 5000 5000>
    componentLinearAccScale=<1 1 1>
    componentLinearVelScale=<1 1 1>
    physicsBodyDefinitions [{
      sphereCollisionRadius=0
      boundBone {boneName=dyn_jacket_back}
      boxExtents=<20 10 10>
      constraintSetup {angularLimitsMax=<25 0 25>, angularLimitsMin=<-25 -25 -25>, angularTarget=<0 0 0>}
      localJointOffset=<2 0 0>
     }]
   }
   n16 = Anim Dynamics - Bone: dyn_jacket_l(ComponentPose=@n15, Alpha=0.4, ExternalForce="0,0,0") {
    angularDampingOverride=1
    angularSpringConstant=75
    bAngularSpring=true
    bGravityOverrideInSimSpace=true
    bOverrideAngularDamping=true
    simulationSpace=RootRelative
    alphaBoolBlend {blendOption=Linear}
    boundBone {boneName=dyn_jacket_l}
    componentAppliedLinearAccClamp=<5000 5000 5000>
    componentLinearAccScale=<1 1 1>
    componentLinearVelScale=<1 1 1>
    physicsBodyDefinitions [{
      sphereCollisionRadius=0
      boundBone {boneName=dyn_jacket_l}
      boxExtents=<20 10 10>
      constraintSetup {angularLimitsMax=<25 0 25>, angularLimitsMin=<-25 -25 -25>, angularTarget=<0 0 0>}
      localJointOffset=<2 0 0>
     }]
   }
   n17 = RigidBody!(ComponentPose=@n16, bAlphaBoolEnabled:bool, ExternalForce) -> (Pose) {
    alpha=0
    alphaInputType=Bool
    bAlphaBoolEnabled=false
    bClampLinearTranslationLimitToRefPose=true
    bOverrideWorldGravity=true
    overridePhysicsAsset=&/F_MED_Sunshine_Jacket_Physics
    simulationSpace=BaseBoneSpace
    alphaBoolBlend {blendOption=Linear}
    baseBoneRef {boneName=root}
    componentAppliedLinearAccClamp=<500 500 1000>
    componentLinearAccScale=<0.2 0.2 0.4>
    overrideWorldGravity {z=-980}
   }
   n18 = "To Float (Boolean)"(InBool=Get bIsCrouching())
   n19 = "Transform (Modify) Bone - Bone: dyn_jacket_back_r"(ComponentPose=@n17, Alpha=@n18, Translation="0,-2,0", Rotation="0,0,0", Scale="1,1,1") {
    rotationMode=BMM_Additive
    scaleMode=BMM_Additive
    translationMode=BMM_Additive
    alphaBoolBlend {blendOption=Linear}
    boneToModify {boneName=dyn_jacket_back_r}
    translation {y=-2}
   }
   n20 = "Transform (Modify) Bone - Bone: dyn_jacket_back"(ComponentPose=@n19, Alpha=@n18, Translation="0,-2,0", Rotation="0,0,0", Scale="1,1,1") {
    rotationMode=BMM_Additive
    scaleMode=BMM_Additive
    translationMode=BMM_Additive
    alphaBoolBlend {blendOption=Linear}
    boneToModify {boneName=dyn_jacket_back}
    translation {y=-2}
   }
   n21 = Anim Dynamics - Bone: dyn_collar_r(ComponentPose=@n20, Alpha=1, ExternalForce="0,0,0") {
    angularDampingOverride=0.9999
    angularSpringConstant=200
    bAngularSpring=true
    bGravityOverrideInSimSpace=true
    bOverrideAngularDamping=true
    gravityScale=0.3
    lODThreshold=1
    alphaBoolBlend {blendOption=Linear}
    boundBone {boneName=dyn_collar_r}
    componentLinearAccScale=<0.25 0.25 0.6>
    physicsBodyDefinitions [{
      sphereCollisionRadius=0
      boundBone {boneName=dyn_collar_r}
      boxExtents=<5 20 20>
      constraintSetup {twistAxis=AxisZ, angularLimitsMax=<10 10 10>, angularLimitsMin=<-10 -10 -10>, angularTarget=<0 0 0>}
      localJointOffset=<0 0 -0.1>
     }]
    relativeSpaceBone {boneName=root}
   }
   n22 = Anim Dynamics - Bone: dyn_collar_back_r(ComponentPose=@n21, Alpha=1, ExternalForce="0,0,0") {
    angularDampingOverride=0.9999
    angularSpringConstant=200
    bAngularSpring=true
    bGravityOverrideInSimSpace=true
    bOverrideAngularDamping=true
    gravityScale=0.3
    lODThreshold=1
    alphaBoolBlend {blendOption=Linear}
    boundBone {boneName=dyn_collar_back_r}
    componentLinearAccScale=<0.25 0.25 0.6>
    physicsBodyDefinitions [{
      sphereCollisionRadius=0
      boundBone {boneName=dyn_collar_back_r}
      boxExtents=<5 20 20>
      constraintSetup {twistAxis=AxisZ, angularLimitsMax=<10 3 10>, angularLimitsMin=<-10 -10 -10>, angularTarget=<0 0 0>}
      localJointOffset=<0 0 -1>
     }]
    relativeSpaceBone {boneName=root}
   }
   n23 = Anim Dynamics - Bone: dyn_collar_back_l(ComponentPose=@n22, Alpha=1, ExternalForce="0,0,0") {
    angularDampingOverride=0.9999
    angularSpringConstant=200
    bAngularSpring=true
    bGravityOverrideInSimSpace=true
    bOverrideAngularDamping=true
    gravityScale=0.3
    lODThreshold=1
    alphaBoolBlend {blendOption=Linear}
    boundBone {boneName=dyn_collar_back_l}
    componentLinearAccScale=<0.25 0.25 0.6>
    physicsBodyDefinitions [{
      sphereCollisionRadius=0
      boundBone {boneName=dyn_collar_back_l}
      boxExtents=<5 20 20>
      constraintSetup {twistAxis=AxisZ, angularLimitsMax=<10 3 10>, angularLimitsMin=<-10 -10 -10>, angularTarget=<0 0 0>}
      localJointOffset=<0 0 -1>
     }]
    relativeSpaceBone {boneName=root}
   }
   n24 = Anim Dynamics - Bone: dyn_collar_l(ComponentPose=@n23, Alpha=1, ExternalForce="0,0,0") {
    angularDampingOverride=0.9999
    angularSpringConstant=200
    bAngularSpring=true
    bGravityOverrideInSimSpace=true
    bOverrideAngularDamping=true
    gravityScale=0.3
    lODThreshold=1
    alphaBoolBlend {blendOption=Linear}
    boundBone {boneName=dyn_collar_l}
    componentLinearAccScale=<0.25 0.25 0.6>
    physicsBodyDefinitions [{
      sphereCollisionRadius=0
      boundBone {boneName=dyn_collar_l}
      boxExtents=<5 20 20>
      constraintSetup {twistAxis=AxisZ, angularLimitsMax=<10 10 10>, angularLimitsMin=<-10 -10 -10>, angularTarget=<0 0 0>}
      localJointOffset=<0 0 -1>
     }]
    relativeSpaceBone {boneName=root}
   }
   n25 = "To Float (Boolean)"(InBool=Get bIsCrouching())
   n26 = "Constraint - dyn_kneepad (PP)"!(ComponentPose=@n24, Alpha=@n25, ConstraintWeights_0=0.25, ConstraintWeights_1:float=1) -> (Pose) {
    alphaBoolBlend {blendOption=Linear}
    boneToModify {boneName=dyn_kneepad}
    constraintSetup [{targetBone {boneName=thigh_twist_01_r}}, {targetBone {boneName=calf_r}}]
    constraintWeights [1,1]
   }
   n27 = "Transform (Modify) Bone - Bone: dyn_kneepad"(ComponentPose=@n26, Alpha=@n25, Translation="0,0,1", Rotation="0,0,-10", Scale="1,1,1") {
    rotationMode=BMM_Additive
    scaleMode=BMM_Additive
    translationMode=BMM_Additive
    alphaBoolBlend {blendOption=Linear}
    boneToModify {boneName=dyn_kneepad}
    rotation {roll=-10}
    translation {z=1}
   }
   n28 = "Constraint - dyn_l_frontShoe (P)"(ComponentPose=@n27, Alpha=1, ConstraintWeights_0=1) {
    alphaBoolBlend {blendOption=Linear}
    boneToModify {boneName=dyn_l_frontShoe}
    constraintSetup [{targetBone {boneName=calf_l}}]
    constraintWeights [1]
   }
   n29 = "Constraint - dyn_r_frontShoe (P)"(ComponentPose=@n28, Alpha=1, ConstraintWeights_0=1) {
    alphaBoolBlend {blendOption=Linear}
    boneToModify {boneName=dyn_r_frontShoe}
    constraintSetup [{targetBone {boneName=calf_r}}]
    constraintWeights [1]
   }
   n30 = RigidBody!(ComponentPose=@n29, bAlphaBoolEnabled, ExternalForce, bTransferBoneVelocities:bool=False) -> (Pose) {
    alphaInputType=Bool
    bAlphaBoolEnabled=false
    bClampLinearTranslationLimitToRefPose=true
    bOverrideWorldGravity=true
    overridePhysicsAsset=&/F_MED_Sunshine_Shoe_Physics
    simulationSpace=BaseBoneSpace
    alphaBoolBlend {blendOption=Linear}
    baseBoneRef {boneName=root}
    overrideWorldGravity {z=-980}
   }
   n31 = "Transform (Modify) Bone - Bone: dyn_r_frontShoe"(ComponentPose=@n30, Alpha=1, Translation="0,0,0", Rotation="0,0,10", Scale="1,1,1") {
    rotationMode=BMM_Additive
    scaleMode=BMM_Additive
    translationMode=BMM_Additive
    alphaBoolBlend {blendOption=Linear}
    boneToModify {boneName=dyn_r_frontShoe}
    rotation {roll=10}
   }
   n32 = "Transform (Modify) Bone - Bone: dyn_l_backShoe"(ComponentPose=@n31, Alpha=1, Translation="0,0,0", Rotation="0,0,10", Scale="1,1,1") {
    rotationMode=BMM_Additive
    scaleMode=BMM_Additive
    translationMode=BMM_Additive
    alphaBoolBlend {blendOption=Linear}
    boneToModify {boneName=dyn_l_backShoe}
    rotation {roll=10}
   }
   n33 = Anim Dynamics - Bone: dyn_l_frontShoe(ComponentPose=@n32, Alpha=0.3, ExternalForce="0,0,0") {
    angularDampingOverride=0.999
    angularSpringConstant=12000
    bAngularSpring=true
    bGravityOverrideInSimSpace=true
    bOverrideAngularDamping=true
    bUseGravityOverride=true
    gravityScale=0.35
    simulationSpace=RootRelative
    alphaBoolBlend {blendOption=Linear}
    boundBone {boneName=dyn_l_frontShoe}
    componentLinearAccScale=<1 1 1>
    componentLinearVelScale=<1 1 1>
    gravityOverride {z=-980}
    physicsBodyDefinitions [{
      sphereCollisionRadius=0
      boundBone {boneName=dyn_l_frontShoe}
      boxExtents=<20 10 10>
      constraintSetup {angularLimitsMax=<0 0 15>, angularLimitsMin=<0 0 -15>, angularTarget=<1 0 0>}
      localJointOffset=<4 0 0>
     }]
   }
   n34 = Anim Dynamics - Bone: dyn_r_frontShoe(ComponentPose=@n33, Alpha=0.3, ExternalForce="0,0,0") {
    angularDampingOverride=0.999
    angularSpringConstant=12000
    bAngularSpring=true
    bGravityOverrideInSimSpace=true
    bOverrideAngularDamping=true
    bUseGravityOverride=true
    gravityScale=0.35
    simulationSpace=RootRelative
    alphaBoolBlend {blendOption=Linear}
    boundBone {boneName=dyn_r_frontShoe}
    componentLinearAccScale=<1 1 1>
    componentLinearVelScale=<1 1 1>
    gravityOverride {z=-980}
    physicsBodyDefinitions [{
      sphereCollisionRadius=0
      boundBone {boneName=dyn_r_frontShoe}
      boxExtents=<20 10 10>
      constraintSetup {angularLimitsMax=<0 0 15>, angularLimitsMin=<0 0 -15>, angularTarget=<1 0 0>}
      localJointOffset=<4 0 0>
     }]
   }
   n35 = Anim Dynamics - Bone: dyn_l_backShoe(ComponentPose=@n34, Alpha=0.4, ExternalForce="0,0,0") {
    angularDampingOverride=0.999
    angularSpringConstant=12000
    bAngularSpring=true
    bGravityOverrideInSimSpace=true
    bOverrideAngularDamping=true
    bUseGravityOverride=true
    gravityScale=0.35
    simulationSpace=RootRelative
    alphaBoolBlend {blendOption=Linear}
    boundBone {boneName=dyn_l_backShoe}
    componentLinearAccScale=<1 1 1>
    componentLinearVelScale=<1 1 1>
    gravityOverride {z=-980}
    physicsBodyDefinitions [{
      sphereCollisionRadius=0
      boundBone {boneName=dyn_l_backShoe}
      boxExtents=<20 10 10>
      constraintSetup {angularLimitsMax=<0 0 15>, angularLimitsMin=<0 0 -15>, angularTarget=<1 0 0>}
      localJointOffset=<4 0 0>
     }]
   }
   n36 = Anim Dynamics - Bone: dyn_r_backShoe(ComponentPose=@n35, Alpha=0.4, ExternalForce="0,0,0") {
    angularDampingOverride=0.999
    angularSpringConstant=12000
    bAngularSpring=true
    bGravityOverrideInSimSpace=true
    bOverrideAngularDamping=true
    bUseGravityOverride=true
    gravityScale=0.35
    simulationSpace=RootRelative
    alphaBoolBlend {blendOption=Linear}
    boundBone {boneName=dyn_r_backShoe}
    componentLinearAccScale=<1 1 1>
    componentLinearVelScale=<1 1 1>
    gravityOverride {z=-980}
    physicsBodyDefinitions [{
      sphereCollisionRadius=0
      boundBone {boneName=dyn_r_backShoe}
      boxExtents=<20 10 10>
      constraintSetup {angularLimitsMax=<0 0 15>, angularLimitsMin=<0 0 -15>, angularTarget=<1 0 0>}
      localJointOffset=<4 0 0>
     }]
   }
   n37 = "Transform (Modify) Bone - Bone: dyn_collar_l"(ComponentPose=@n36, Alpha=1, Translation="0,-2,0", Rotation="0,0,0", Scale="1,1,1") {
    rotationMode=BMM_Additive
    scaleMode=BMM_Additive
    translationMode=BMM_Additive
    alphaBoolBlend {blendOption=Linear}
    boneToModify {boneName=dyn_collar_l}
    translation {y=-2}
   }
   n38 = "Constraint - dyn_neck (PPP)"!(ComponentPose=@n37, Alpha=1, ConstraintWeights_0=0.5, ConstraintWeights_1:float=0.5, ConstraintWeights_2:float=0.4) -> (Pose) {
    alphaBoolBlend {blendOption=Linear}
    boneToModify {boneName=dyn_neck}
    constraintSetup [{targetBone {boneName=neck_02}}, {targetBone {boneName=neck_01}}, {targetBone {boneName=head}}]
    constraintWeights [1,1,1]
   }
   n39 = "Transform (Modify) Bone - Bone: dyn_neck"(ComponentPose=@n38, Alpha=1, Translation="0,0,0", Rotation="0,0,0", Scale="1.3,1.1,1.3") {
    rotationMode=BMM_Additive
    scaleMode=BMM_Additive
    translationMode=BMM_Additive
    alphaBoolBlend {blendOption=Linear}
    boneToModify {boneName=dyn_neck}
    scale=<1.3 1.1 1.3>
   }
   n40 = "Transform (Modify) Bone - Bone: dyn_chest"(ComponentPose=@n39, Alpha=1, Translation="6,0,0", Rotation="0,0,0", Scale="1.22,1.22,1.22") {
    rotationMode=BMM_Additive
    rotationSpace=BCS_BoneSpace
    scaleMode=BMM_Additive
    scaleSpace=BCS_BoneSpace
    translationMode=BMM_Additive
    translationSpace=BCS_BoneSpace
    alphaBoolBlend {blendOption=Linear}
    boneToModify {boneName=dyn_chest}
    scale=<1.22 1.22 1.22>
    translation {x=6}
   }
   n41 = Control Rig(Source=Component To Local(ComponentPose=@n40), Alpha=SelectFloat(A=1, B=0.25, bPickA=Get bIsCrouching())) {
    alphaBoolBlend {blendOption=Linear}
    controlRigAssetReference {blueprintRigClass=&/F_MED_Sunshine_Ctrl_Rig.F_MED_Sunshine_Ctrl_Rig_C}
   }
   Output Pose(Result=@n41) {layerGroup=None}
  ]}
 EventGraph {nodes [
   Event BlueprintUpdateAnimation()
   > n1 = Sequence!(execute) -> (then_0, then_1, then_2:Exec, then_3:Exec)
   > n2 = Branch(Condition=OR Boolean!(A=Get bIsJumping(), B=Get bIsFalling(), C:bool=Get bIsCrouching()) -> (ReturnValue))
   > Set thigh_rot_r(thigh_rot_r=0)
   > Set thigh_rot_l(thigh_rot_l=0)
   n3 = BreakTransform(InTransform=GetDeltaTransformFromRefPose(self=GetOwnerSkeletalMeshComponent(), BoneName=thigh_l, BaseName=pelvis))
   > Set Thigh L Rot Crouch(Thigh L Rot Crouch:double=MapRangeClamped(Value=BreakRotator(InRot=@n3.Rotation).Yaw, InRangeA=-40, InRangeB=-100)) -> (Output_Get:double)
   n4 = BreakTransform(InTransform=GetDeltaTransformFromRefPose(self=GetOwnerSkeletalMeshComponent(), BoneName=thigh_r, BaseName=pelvis))
   > Set Thigh_R_Rot_Crouch(Thigh_R_Rot_Crouch:double=MapRangeClamped(Value=BreakRotator(InRot=@n4.Rotation).Yaw, InRangeA=-40, InRangeB=-100)) -> (Output_Get:double)
   Set Thigh L Rot Crouch(execute=@n2.else, Thigh L Rot Crouch=0)
   > Set Thigh_R_Rot_Crouch(Thigh_R_Rot_Crouch=0)
   n5 = Get Skel_Mesh()
   > Set thigh_rot_r(thigh_rot_r=GetAlphaFromReferenceTransformDelta(SkeletalMeshComponent=@n5, BoneName=thigh_r, BaseName=pelvis, InputRange=(X=0,Y=-90), TwistAxis="0,0,1"))
   > Set thigh_rot_l(thigh_rot_l=GetAlphaFromReferenceTransformDelta(SkeletalMeshComponent=@n5, BoneName=thigh_l, BaseName=pelvis, InputRange=(X=-20,Y=-101), TwistAxis="0,0,1"))
   n6 = Branch(execute=@n1.then_1, Condition=AND Boolean(A=Get bIsInFrontEnd(), B=Get bIsPlayingEmote()))
   > Set FrontEndEmoting(FrontEndEmoting:bool=true) -> (Output_Get:bool)
   Set FrontEndEmoting(execute=@n6.else, FrontEndEmoting=false)
   n7 = Branch(execute=@n1.then_2, Condition=Get bIsCrouching())
   > Set Crouching(Crouching:bool=true) -> (Output_Get:bool)
   Set Crouching(execute=@n7.else, Crouching=true)
   n8 = Gravity Override State Function DEPRECATED(
    execute=@n1.then_3
    CharacterPartAnimInstance=Self-Reference()
    GameTimeInSeconds=GetGameTimeInSeconds()
    GravityOverrideParams=Get Gravity Override Params()
   )
   > Set GravityOverride(GravityOverride=@n8)
   Event BlueprintInitializeAnimation()
   > Set Skel_Mesh(Skel_Mesh=GetOwnerSkeletalMeshComponent())
   n9 = Event OnBeginCrouching()
   ResetDynamics(execute=[@n9.then, @n10.then], InTeleportType=ResetPhysics)
   > ResetDynamics(InTeleportType=TeleportPhysics)
   n10 = Event OnEndCrouching()
  ]}
]
variables [
 Crouching {type=boolean}
 FrontEndEmoting {type=boolean}
 Gravity Override Params {
  type=GravityOverrideParamsStruct
  default {
   emoteOrMelee {gravityOverride=<0 0 -1100>}
   front_End {gravityOverride=<0 0 -1100>}
   onGround_Moving {gravityOverride=<0 0 -1100>}
   onGround_Standing {gravityOverride=<0 0 -1100>}
   skydive_DivingDown {gravityOverride=<0 0 800>}
   skydive_FloatingDown {gravityOverride=<0 0 600>}
   skydive_GoingUp {gravityOverride=<0 0 100>}
   skydive_Parachuting {gravityOverride=<0 0 -1100>}
   swimming {gravityOverride=<0 0 0>}
  }
 }
 GravityOverride {type=Vector}
 Skel_Mesh {type=SkeletalMeshComponent}
 Thigh L Rot Crouch {type=number}
 Thigh_R_Rot_Crouch {type=number}
 thigh_rot_l {type=number}
 thigh_rot_r {type=number}
]
