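# Control Rig for the Sunshine female character's face accessory part, adding a procedural correction to the main hair bone dyn_mainHair_02.
# It reads head relative to neck_01, remaps the transform's Z location with clamping, and applies it as an additive local X translation on that bone.
# A second branch remaps head Euler Z into an X rotation on the same bone, but its modify node is not connected to execution.
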ControlRigBlueprint /BRCosmetics/Characters/Player/Female/Medium/Bodies/F_MED_Sunshine/Meshes/Parts/F_MED_Sunshine_FaceAcc_CtrlRig
graphs [
 RigVMFunctionLibrary {links [], nodes []}
 RigVMModel {nodes [
   RigUnit_BeginExecution = &RigUnit_BeginExecution()
   n1 = /Script/RigVM.RigVMFunction_MathTransformToEulerTransform(Value=&RigUnit_GetRelativeBoneTransform(Bone=head, Space=neck_01).Transform) {is=struct RigVMUnitNode, template="Cast::Execute(in Value,out Result)"}
   /Script/ControlRig.RigUnit_ModifyBoneTransforms(
    BoneToModify=((Bone="dyn_mainHair_02",Transform=(Rotation=(X=0,Y=0,Z=0,W=1),Translation=(X=0,Y=0,Z=0),Scale3D=(X=1,Y=1,Z=1))))
    Mode=AdditiveLocal
    Weight=1
    WeightMaximum=1
    WeightMinimum=0
    BoneToModify.0.Transform.Translation.X=&RigVMFunction_MathFloatRemap(SourceMaximum=-50, SourceMinimum=-25, TargetMaximum=50, TargetMinimum=0, bClamp=true, Value=@n1.Result.Location.Z).Result
    ExecutePin=@RigUnit_BeginExecution.ExecutePin
   ) {is=struct RigVMUnitNode}
   &RigVMFunction_MathFloatRemap(SourceMaximum=-50, SourceMinimum=-25, TargetMaximum=50, TargetMinimum=0, bClamp=true, Value=@n1.Result.Location.Z)
   n2 = &RigVMFunction_MathQuaternionToEuler(RotationOrder=ZYX, Value=&RigUnit_GetRelativeBoneTransform(Bone=head, Space=neck_01).Transform.Rotation)
   &RigUnit_ModifyBoneTransforms(
    BoneToModify=((Bone="dyn_mainHair_02",Transform=(Rotation=(X=0,Y=0,Z=0,W=1),Translation=(X=0,Y=0,Z=0),Scale3D=(X=1,Y=1,Z=1))))
    ExecutePin=()
    Mode=AdditiveLocal
    Weight=1
    WeightMaximum=1
    WeightMinimum=0
    BoneToModify.0.Transform.Rotation=&RigVMFunction_MathQuaternionFromEuler(
     Euler=(X=0,Y=0,Z=0)
     Euler.X=&RigVMFunction_MathFloatRemap(SourceMaximum=-50, SourceMinimum=-25, TargetMaximum=6, TargetMinimum=0, bClamp=true, Value=@n2.Result.Z).Result
    ).Result
   )
  ]}
]
hierarchy [
 root {loc=<0 0 0>, rot=<0 0 0 1>}
 attach {parent=root, loc=<0 0 0>, rot=<0 0 0 1>}
 pelvis {parent=root, loc=<0 -0.5301 98.0571>, rot=<0.0127 0.707 -0.0127 -0.707>}
 spine_01 {parent=pelvis, loc=<6.1208 -0.9328 0>, rot=<0 0 0.034 -0.9994>}
 spine_02 {parent=spine_01, loc=<6.5996 -0.2095 0>, rot=<0 0 0.0238 -0.9997>}
 spine_03 {parent=spine_02, loc=<6.4914 -0.1237 0>, rot=<0 0 0.0149 -0.9999>}
 spine_04 {parent=spine_03, loc=<6.9598 0.0794 0>, rot=<0 0 -0.0132 -0.9999>}
 spine_05 {parent=spine_04, loc=<6.9987 0 0>, rot=<0 0 -0.0015 -1>}
 clavicle_l {parent=spine_05, loc=<5.4831 -0.1053 -3.2946>, rot=<0.0465 -0.7289 -0.009 -0.6829>}
 upperarm_l {parent=clavicle_l, loc=<9.8658 -1.2136 -0.4105>, rot=<0.0208 -0.486 -0.0242 -0.8734>}
 upperarm_twist_01_l {parent=upperarm_l, loc=<7.2135 -0.1134 0>, rot=<0 0 0 1>}
 upperarm_twist_02_l {parent=upperarm_l, loc=<19.7998 -0.1134 0>, rot=<0 0 0 1>}
 clavicle_r {parent=spine_05, loc=<5.4831 -0.1053 3.2946>, rot=<0.7289 0.0465 -0.6829 0.009>}
 upperarm_r {parent=clavicle_r, loc=<-9.8658 1.2136 0.4104>, rot=<0.0208 -0.486 -0.0242 -0.8734>}
 upperarm_twist_01_r {parent=upperarm_r, loc=<-7.2135 0.1134 0>, rot=<0 0 0 1>}
 upperarm_twist_02_r {parent=upperarm_r, loc=<-19.7998 0.1134 0>, rot=<0 0 0 1>}
 neck_01 {parent=spine_05, loc=<11.3138 0.1529 0>, rot=<0 0 -0.2003 -0.9797>}
 neck_02 {parent=neck_01, loc=<4.5627 0 0>, rot=<0 0 0.0504 -0.9987>}
 head {parent=neck_02, loc=<5.3914 0 0>, rot=<0 0 0.115 -0.9934>}
 dyn_hairBang_a_01 {parent=head, loc=<10.0347 7.8952 3.3284>, rot=<0.1898 0.4443 0.8731 -0.066>}
 dyn_hairBang_a_02 {parent=dyn_hairBang_a_01, loc=<2.4087 0 0>, rot=<-0.0478 -0.2311 -0.1497 -0.9601>}
 dyn_hairBang_a_03 {parent=dyn_hairBang_a_02, loc=<2.5733 0 0>, rot=<0.0348 -0.1202 -0.2237 -0.9666>}
 dyn_hairBang_b_01 {parent=head, loc=<9.7666 8.723 -0.9032>, rot=<-0.1812 -0.712 -0.6761 0.0564>}
 dyn_hairBang_b_02 {parent=dyn_hairBang_b_01, loc=<2.132 0 0>, rot=<-0.0291 -0.0771 -0.0914 -0.9924>}
 dyn_hairBang_b_03 {parent=dyn_hairBang_b_02, loc=<2.2705 0 0>, rot=<0.0709 -0.1505 0.0376 -0.9853>}
 dyn_hairBang_c_01 {parent=head, loc=<10.0283 7.4571 -4.7622>, rot=<-0.1199 -0.7248 -0.6746 -0.0717>}
 dyn_hairBang_c_02 {parent=dyn_hairBang_c_01, loc=<2.132 0 0>, rot=<-0.0291 -0.0771 -0.0914 -0.9924>}
 dyn_hairBang_c_03 {parent=dyn_hairBang_c_02, loc=<2.2705 0 0>, rot=<0.0709 -0.1505 0.0376 -0.9853>}
 dyn_earing {parent=head, loc=<0.4429 -0.1021 7.121>, rot=<0.6917 0.1915 0.6765 0.165>}
 dyn_mstr_mainhair {parent=head, loc=<0.6545 0.755 -7.4246>, rot=<0.0029 -0.7071 -0.0029 -0.7071>}
 dyn_mainHair_01 {parent=dyn_mstr_mainhair, loc=<0 0 0>, rot=<-0.0598 -0.7379 -0.1179 -0.6618>}
 dyn_mainHair_02 {parent=dyn_mainHair_01, loc=<5.4151 0 0>, rot=<0 0 0 1>}
 dyn_mainHair_03 {parent=dyn_mainHair_02, loc=<5.1797 0.2625 0.1077>, rot=<0.0102 -0.008 -0.0452 -0.9989>}
 dyn_mainHair_04 {parent=dyn_mainHair_03, loc=<6.1209 0.3234 0.0287>, rot=<-0.0538 0.0258 -0.1748 -0.9828>}
 dyn_mainHair_05 {parent=dyn_mainHair_04, loc=<8.7848 0 0>, rot=<0.5913 0.3949 0.5038 -0.4905>}
 dyn_mainHair_06 {parent=dyn_mainHair_05, loc=<4.4084 0 0>, rot=<0.1222 -0.0678 0.3027 -0.9428>}
 dyn_mainHair_07 {parent=dyn_mainHair_04, loc=<5.6365 -0.3837 2.494>, rot=<0.39 0.4794 0.7341 -0.2814>}
 dyn_mainHair_08 {parent=dyn_mainHair_07, loc=<5.1626 -0.3017 0.1995>, rot=<-0.0886 0.1811 0.2901 -0.9355>}
 dyn_hairback {parent=head, loc=<1.2714 -5.5341 0>, rot=<0.0029 -0.7071 -0.0029 -0.7071>}
 dyn_hairback_02 {parent=head, loc=<-2.0631 -3.3893 -3.5107>, rot=<-0.3821 -0.6917 -0.1567 -0.5925>}
 dyn_collar_l {parent=spine_05, loc=<10.317 1.891 -6.878>, rot=<-0.6763 -0.0343 0.7147 0.1752>}
]
