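# Physics asset for the earring face accessory part of the F_MED_Sunshine female player cosmetic body.
# It defines two capsule bodies, a small one on the dyn_earing bone and a larger one on the head bone.
# A single constraint joins the earring to the head with limited swing and twist ranges, bounding how far the earring can rotate relative to the head.
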PhysicsAsset /BRCosmetics/Characters/Player/Female/Medium/Bodies/F_MED_Sunshine/Meshes/Parts/F_MED_Sunshine_FaceAcc_Earing_Physics
bodies {
 dyn_earing [{
   shapeType=Capsule
   length=1.3286
   radius=2.0786
   shapeName=dyn_earing_capsule
   center=<-0.0776 -0.1168 -1.7373>
   rotation=<rot 82.7099 -0.2154 89.7828>
  }]
 head [{shapeType=Capsule, length=9.5894, radius=7.6617, center=<-0.727 1.762 -2.3021>, rotation=<rot 90 0 0>}]
}
constraints [{bone1Name=dyn_earing, bone2Name=head, swing1LimitDegrees=30, swing2LimitDegrees=30, twistLimitDegrees=90}]
