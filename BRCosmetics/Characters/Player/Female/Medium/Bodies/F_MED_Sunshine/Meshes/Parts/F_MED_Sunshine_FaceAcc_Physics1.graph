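# Physics asset for the Sunshine female medium character's FaceAcc mesh part, setting up capsule bodies and joints for dynamic hair.
# A main hair chain hangs from the head with per-joint swing and twist limits, forking at bone 04 into a 07–08 branch and a 05–06 branch whose root is locked.
# A tightly limited back hair bone also joins the head.
# Spine, neck, head and left-side collar, clavicle and upper arm capsules use free angular joints.
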PhysicsAsset /BRCosmetics/Characters/Player/Female/Medium/Bodies/F_MED_Sunshine/Meshes/Parts/F_MED_Sunshine_FaceAcc_Physics1
bodies {
 clavicle_l [{shapeType=Capsule, length=5.755, radius=3.505, shapeName=clavicle_l_capsule, center=<3.9401 0 0>, rotation=<rot 90 0 0>}]
 dyn_collar_l [{
   shapeType=Capsule
   length=14.755
   radius=2.255
   shapeName=dyn_collar_l_capsule
   center=<-1.4675 -3.4868 0.8309>
   rotation=<rot 22.5191 -45.9098 90.3546>
  }]
 dyn_hairback_02 [{
   shapeType=Capsule
   length=3.505
   radius=2.505
   shapeName=dyn_hairback_02_capsule
   center=<0.9102 0 -0.5415>
   rotation=<rot 7.5878 -102.9392 -60.1123>
  }]
 dyn_mainHair_01 [{shapeType=Capsule, length=4.7622, radius=3.8182, shapeName=dyn_mainHair_01_capsule}]
 dyn_mainHair_02 [{shapeType=Capsule, length=1.25, radius=2.616, shapeName=dyn_mainHair_02_capsule, center=<1.4172 0 -0.0001>, rotation=<rot 0 0 -10.0029>}]
 dyn_mainHair_03 [{shapeType=Capsule, length=0, radius=3.0923, center=<1.3341 0 0>, rotation=<rot 0.1489 1.728 -9.8483>}]
 dyn_mainHair_04 [{shapeType=Capsule, length=2.26, radius=2.999, shapeName=dyn_mainHair_04_capsule, rotation=<rot 20.9818 0 0>}]
 dyn_mainHair_05 [{shapeType=Capsule, length=1.8359, radius=2.7976, shapeName=dyn_mainHair_05_capsule, rotation=<rot 90 0 0>}]
 dyn_mainHair_06 [{shapeType=Capsule, length=2.005, radius=2.255, shapeName=dyn_mainHair_06_capsule, center=<0.5597 0 -0.3815>, rotation=<rot 70 0 0>}]
 dyn_mainHair_07 [{shapeType=Capsule, length=3.4893, radius=3.0823, shapeName=dyn_mainHair_07_capsule, rotation=<rot 90 0 0>}]
 dyn_mainHair_08 [{
   shapeType=Capsule
   length=2.255
   radius=2.505
   shapeName=dyn_mainHair_08_capsule
   center=<1.7012 -0.422 -0.6826>
   rotation=<rot 90 0 18.5298>
  }]
 head [{shapeType=Capsule, length=5.6369, radius=5.1205, rotation=<rot 90 0 49.9994>}]
 neck_01 [{shapeType=Capsule, length=4.75, radius=5.8991, shapeName=neck_01_capsule, rotation=<rot 90 0 0.0001>}]
 neck_02 [{shapeType=Capsule, length=3.7653, radius=5.8123, shapeName=neck_02_capsule}]
 spine_03 [{shapeType=Capsule, length=8.3873, radius=10.5682, shapeName=spine_03_capsule, center=<2.8333 0.2873 0>}]
 spine_04 [{shapeType=Capsule, length=7.2884, radius=13.6454, shapeName=spine_04_capsule, center=<2.8333 0.2873 0>}]
 upperarm_l [{shapeType=Capsule, length=19, radius=6.005, shapeName=upperarm_l_capsule, center=<16.7599 0 0.0001>, rotation=<rot 90 0 0>}]
}
constraints [
 {bone1Name=dyn_mainHair_03, bone2Name=dyn_mainHair_02, swing1LimitDegrees=60, swing2LimitDegrees=10, twistLimitDegrees=5}
 {
  bone1Name=head
  bone2Name=neck_02
  swing1LimitDegrees=45
  swing1Motion=Free
  swing2LimitDegrees=45
  swing2Motion=Free
  twistLimitDegrees=45
  twistMotion=Free
 }
 {
  bone1Name=spine_04
  bone2Name=spine_03
  swing1LimitDegrees=45
  swing1Motion=Free
  swing2LimitDegrees=45
  swing2Motion=Free
  twistLimitDegrees=45
  twistMotion=Free
 }
 {
  bone1Name=neck_01
  bone2Name=spine_04
  swing1LimitDegrees=45
  swing1Motion=Free
  swing2LimitDegrees=45
  swing2Motion=Free
  twistLimitDegrees=45
  twistMotion=Free
 }
 {
  bone1Name=neck_02
  bone2Name=neck_01
  swing1LimitDegrees=45
  swing1Motion=Free
  swing2LimitDegrees=45
  swing2Motion=Free
  twistLimitDegrees=45
  twistMotion=Free
 }
 {bone1Name=dyn_mainHair_02, bone2Name=dyn_mainHair_01, swing1LimitDegrees=40, swing2LimitDegrees=10, twistLimitDegrees=3}
 {bone1Name=dyn_mainHair_01, bone2Name=head, swing1LimitDegrees=35, swing2LimitDegrees=10, twistLimitDegrees=3}
 {bone1Name=dyn_mainHair_04, bone2Name=dyn_mainHair_03, swing1LimitDegrees=40, swing2LimitDegrees=15, twistLimitDegrees=5}
 {bone1Name=dyn_mainHair_07, bone2Name=dyn_mainHair_04, swing1LimitDegrees=20, swing2LimitDegrees=12, twistLimitDegrees=3}
 {bone1Name=dyn_mainHair_08, bone2Name=dyn_mainHair_07, swing1LimitDegrees=10, swing2LimitDegrees=10, twistLimitDegrees=3}
 {
  bone1Name=dyn_mainHair_05
  bone2Name=dyn_mainHair_04
  swing1LimitDegrees=5
  swing1Motion=Locked
  swing2LimitDegrees=5
  swing2Motion=Locked
  twistLimitDegrees=5
  twistMotion=Locked
 }
 {bone1Name=dyn_mainHair_06, bone2Name=dyn_mainHair_05, swing1LimitDegrees=10, swing2LimitDegrees=10, twistLimitDegrees=10}
 {bone1Name=dyn_hairback_02, bone2Name=head, swing1LimitDegrees=5, swing2LimitDegrees=5, twistLimitDegrees=5}
 {
  bone1Name=dyn_collar_l
  bone2Name=spine_04
  swing1LimitDegrees=45
  swing1Motion=Free
  swing2LimitDegrees=45
  swing2Motion=Free
  twistLimitDegrees=45
  twistMotion=Free
 }
 {
  bone1Name=clavicle_l
  bone2Name=spine_04
  swing1LimitDegrees=45
  swing1Motion=Free
  swing2LimitDegrees=45
  swing2Motion=Free
  twistLimitDegrees=45
  twistMotion=Free
 }
 {
  bone1Name=upperarm_l
  bone2Name=clavicle_l
  swing1LimitDegrees=45
  swing1Motion=Free
  swing2LimitDegrees=45
  swing2Motion=Free
  twistLimitDegrees=45
  twistMotion=Free
 }
]
