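# Physics asset for the bangs face accessory of the female medium Sunshine player body, defining physical motion for the hair.
# Three bang chains (a, b, c) each have three small capsule bodies, with each root constrained to a larger head capsule.
# Each segment is constrained to its parent with tight, uniform swing and even tighter twist limits, allowing only slight movement per joint.
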PhysicsAsset /Game/Characters/Player/Female/Medium/Bodies/F_MED_Sunshine/Meshes/Parts/F_MED_Sunshine_FaceAcc_Bangs_Physics
bodies {
 dyn_hairBang_a_01 [{shapeType=Capsule, length=1.8974, radius=1.25, shapeName=dyn_hairBang_a_01_capsule, center=<0.3824 0 0>, rotation=<rot 0 0 90>}]
 dyn_hairBang_a_02 [{shapeType=Capsule, length=2.7492, radius=1.0328, shapeName=dyn_hairBang_a_02_capsule, center=<0.3144 0 0>, rotation=<rot 0 0 89.9999>}]
 dyn_hairBang_a_03 [{
   shapeType=Capsule
   length=1.005
   radius=1.005
   shapeName=dyn_hairBang_a_03_capsule
   center=<0.3201 -0.1165 0>
   rotation=<rot 0 -19.9995 89.9999>
  }]
 dyn_hairBang_b_01 [{shapeType=Capsule, length=2.3045, radius=1.4391, shapeName=dyn_hairBang_b_01_capsule, center=<0.4991 0 0>, rotation=<rot 0 0 89.9999>}]
 dyn_hairBang_b_02 [{shapeType=Capsule, length=3.0513, radius=1, shapeName=dyn_hairBang_b_02_capsule, center=<0.4991 0 0>, rotation=<rot 0 0 89.9999>}]
 dyn_hairBang_b_03 [{shapeType=Capsule, length=1.005, radius=1.005, shapeName=dyn_hairBang_b_03_capsule, center=<0.4991 0 0>, rotation=<rot 0 0 89.9999>}]
 dyn_hairBang_c_01 [{shapeType=Capsule, length=2.3064, radius=1.472, shapeName=dyn_hairBang_c_01_capsule, center=<0.4991 0 0>, rotation=<rot 0 0 89.9999>}]
 dyn_hairBang_c_02 [{shapeType=Capsule, length=2.3334, radius=1, shapeName=dyn_hairBang_c_02_capsule, center=<0.3896 -0.552 0>, rotation=<rot 0 0 89.9999>}]
 dyn_hairBang_c_03 [{shapeType=Capsule, length=1.005, radius=1.005, shapeName=dyn_hairBang_c_03_capsule, center=<0.4991 0 0>, rotation=<rot 0 0 89.9999>}]
 head [{shapeType=Capsule, length=13.2655, radius=8.2853, rotation=<rot 90 0 0>}]
}
constraints [
 {bone1Name=dyn_hairBang_a_01, bone2Name=head, swing1LimitDegrees=5, swing2LimitDegrees=5, twistLimitDegrees=1}
 {bone1Name=dyn_hairBang_a_02, bone2Name=dyn_hairBang_a_01, swing1LimitDegrees=5, swing2LimitDegrees=5, twistLimitDegrees=1}
 {bone1Name=dyn_hairBang_a_03, bone2Name=dyn_hairBang_a_02, swing1LimitDegrees=5, swing2LimitDegrees=5, twistLimitDegrees=1}
 {bone1Name=dyn_hairBang_b_01, bone2Name=head, swing1LimitDegrees=5, swing2LimitDegrees=5, twistLimitDegrees=1}
 {bone1Name=dyn_hairBang_b_02, bone2Name=dyn_hairBang_b_01, swing1LimitDegrees=5, swing2LimitDegrees=5, twistLimitDegrees=1}
 {bone1Name=dyn_hairBang_b_03, bone2Name=dyn_hairBang_b_02, swing1LimitDegrees=5, swing2LimitDegrees=5, twistLimitDegrees=1}
 {bone1Name=dyn_hairBang_c_01, bone2Name=head, swing1LimitDegrees=5, swing2LimitDegrees=5, twistLimitDegrees=1}
 {bone1Name=dyn_hairBang_c_02, bone2Name=dyn_hairBang_c_01, swing1LimitDegrees=5, swing2LimitDegrees=5, twistLimitDegrees=1}
 {bone1Name=dyn_hairBang_c_03, bone2Name=dyn_hairBang_c_02, swing1LimitDegrees=5, swing2LimitDegrees=5, twistLimitDegrees=1}
]
