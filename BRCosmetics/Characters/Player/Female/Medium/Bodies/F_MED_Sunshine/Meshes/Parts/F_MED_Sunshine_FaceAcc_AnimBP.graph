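# Post-processes the Sunshine female cosmetic's face-accessory mesh, adding secondary motion to hair, bangs and earring.
# Copied parent and Body_Mesh poses are layered, then fed through Control Rig, RigidBody hair and earring sims, and per-bone AnimDynamics bangs.
# Weak constraints toward spine and clavicle and an additive earring offset finish the pose.
# An older RigidBody chain and NeckLookUp logic are unconnected; bang a_03 is simulated twice, b_03 never.
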AnimBlueprint /BRCosmetics/Characters/Player/Female/Medium/Bodies/F_MED_Sunshine/Meshes/Parts/F_MED_Sunshine_FaceAcc_AnimBP
target_skeleton=&/F_MED_Sunshine_FaceAcc_Skeleton
graphs [
 AnimGraph {nodes [
   n1 = RigidBody(ComponentPose=Local To Component(LocalPose=Copy Pose From Mesh() {bCopyCurves=true, bUseAttachedParent=true})) {
    bOverrideWorldGravity=true
    overridePhysicsAsset=/Game/Characters/Player/Female/Medium/Bodies/F_MED_Sunshine/Meshes/Parts/F_MED_Sunshine_FaceAcc_Physics
    simulationSpace=BaseBoneSpace
    alphaBoolBlend {blendOption=Linear}
    baseBoneRef {boneName=neck_01}
    componentAppliedLinearAccClamp=<2000 2000 2000>
    componentLinearAccScale=<0.3 0.35 0.6>
    componentLinearVelScale=<0.3 0.35 0.4>
    overrideWorldGravity {z=-980}
   }
   n2 = RigidBody(ComponentPose=@n1) {
    bClampLinearTranslationLimitToRefPose=true
    bOverrideWorldGravity=true
    overridePhysicsAsset=/Game/Characters/Player/Female/Medium/Bodies/F_MED_Sunshine/Meshes/Parts/F_MED_Sunshine_FaceAcc_Bangs_Physics
    simulationSpace=BaseBoneSpace
    alphaBoolBlend {blendOption=Linear}
    baseBoneRef {boneName=neck_01}
    componentAppliedLinearAccClamp=<5000 5000 5000>
    componentLinearAccScale=<0.45 0.45 0.45>
    overrideWorldGravity {z=-980}
   }
   n3 = RigidBody(ComponentPose=@n2) {
    bClampLinearTranslationLimitToRefPose=true
    bOverrideWorldGravity=true
    overridePhysicsAsset=&/F_MED_Sunshine_FaceAcc_Earing_Physics
    alphaBoolBlend {blendOption=Linear}
    componentAppliedLinearAccClamp=<2500 2500 2500>
    componentLinearAccScale=<0.01 0.01 0.01>
    overrideWorldGravity {z=-1100}
   }
   n4 = "Constraint - dyn_mainHair_03 (P)"(ComponentPose=@n3, Alpha=0.25, ConstraintWeights_0=1) {
    alphaBoolBlend {blendOption=Linear}
    boneToModify {boneName=dyn_mainHair_03}
    constraintSetup [{targetBone {boneName=clavicle_l}}]
    constraintWeights [1]
   }
   n5 = "Constraint - dyn_mainHair_04 (P)"(ComponentPose=@n4, Alpha=0.25, ConstraintWeights_0=1) {
    alphaBoolBlend {blendOption=Linear}
    boneToModify {boneName=dyn_mainHair_04}
    constraintSetup [{targetBone {boneName=spine_05}}]
    constraintWeights [1]
   }
   n6 = "Constraint - dyn_mainHair_05 (P)"(ComponentPose=@n5, Alpha=0.25, ConstraintWeights_0=1) {
    alphaBoolBlend {blendOption=Linear}
    boneToModify {boneName=dyn_mainHair_05}
    constraintSetup [{targetBone {boneName=spine_05}}]
    constraintWeights [1]
   }
   n7 = "Constraint - dyn_mainHair_06 (P)"(ComponentPose=@n6, Alpha=0.25, ConstraintWeights_0=1) {
    alphaBoolBlend {blendOption=Linear}
    boneToModify {boneName=dyn_mainHair_06}
    constraintSetup [{targetBone {boneName=spine_05}}]
    constraintWeights [1]
   }
   "Constraint - dyn_mainHair_07 (P)"(ComponentPose=@n7, Alpha=0.25, ConstraintWeights_0=1) {
    alphaBoolBlend {blendOption=Linear}
    boneToModify {boneName=dyn_mainHair_07}
    constraintSetup [{targetBone {boneName=spine_05}}]
    constraintWeights [1]
   }
   Control Rig() {
    alphaBoolBlend {blendOption=Linear}
    controlRigAssetReference {blueprintRigClass=&/F_MED_Sunshine_FaceAcc_CtrlRig.F_MED_Sunshine_FaceAcc_CtrlRig_C}
   }
   "Transform (Modify) Bone - Bone: dyn_mainHair_04"(Alpha=1, Translation="0,-0.3,0", Rotation="-10,0,-5", Scale="1,1,1") {
    rotationMode=BMM_Additive
    translationMode=BMM_Additive
    alphaBoolBlend {blendOption=Linear}
    boneToModify {boneName=dyn_mainHair_04}
    rotation {pitch=-10, roll=-5}
    translation {y=-0.3}
   }
   RigidBody() {
    bOverrideWorldGravity=true
    overridePhysicsAsset=&F_MED_Sunshine_FaceAcc_Physics
    simulationSpace=BaseBoneSpace
    alphaBoolBlend {blendOption=Linear}
    baseBoneRef {boneName=root}
    componentAppliedLinearAccClamp=<5000 5000 5000>
    componentLinearAccScale=<0.3 0.35 0.5>
    componentLinearVelScale=<0.5 0.5 0.5>
    overrideWorldGravity {z=-500}
   }
   RigidBody() {
    bClampLinearTranslationLimitToRefPose=true
    bOverrideWorldGravity=true
    overridePhysicsAsset=&F_MED_Sunshine_FaceAcc_Bangs_Physics
    simulationSpace=BaseBoneSpace
    alphaBoolBlend {blendOption=Linear}
    baseBoneRef {boneName=neck_01}
    componentAppliedLinearAccClamp=<5000 5000 5000>
    componentLinearAccScale=<0.45 0.45 0.45>
    overrideWorldGravity {z=-980}
   }
   n8 = Get Body_Mesh()
   n9 = Copy Pose From Mesh() {bCopyCurves=true, bUseAttachedParent=true}
   n10 = Copy Pose From Mesh(SourceMeshComponent=Get Body_Mesh()) {bCopyCurves=true, bUseAttachedParent=true}
   n11 = Control Rig(Source=Layered blend per bone(
     BasePose=Layered blend per bone(BasePose=@n9, BlendPoses_0=@n10) {layerSetup [{branchFilters [{boneName=upperarm_l}]}]}
     BlendPoses_0=Copy Pose From Mesh(SourceMeshComponent=@n8) {bCopyCurves=true, bUseAttachedParent=true}
    ) {layerSetup [{branchFilters [{boneName=dyn_collar_l}]}]}) {
    alphaBoolBlend {blendOption=Linear}
    controlRigAssetReference {blueprintRigClass=&F_MED_Sunshine_FaceAcc_CtrlRig_C}
   }
   n12 = RigidBody(ComponentPose=Local To Component(LocalPose=@n11)) {
    bOverrideWorldGravity=true
    overridePhysicsAsset=&/F_MED_Sunshine_FaceAcc_Physics1
    simulationSpace=BaseBoneSpace
    alphaBoolBlend {blendOption=Linear}
    baseBoneRef {boneName=neck_01}
    componentAppliedLinearAccClamp=<2000 2000 2000>
    componentLinearAccScale=<0.3 0.35 0.6>
    componentLinearVelScale=<0.3 0.35 0.4>
    overrideWorldGravity {z=-980}
   }
   n13 = RigidBody(ComponentPose=@n12) {
    bClampLinearTranslationLimitToRefPose=true
    bOverrideWorldGravity=true
    overridePhysicsAsset=&F_MED_Sunshine_FaceAcc_Earing_Physics
    simulationSpace=BaseBoneSpace
    alphaBoolBlend {blendOption=Linear}
    baseBoneRef {boneName=head}
    componentAppliedLinearAccClamp=<2500 2500 2500>
    componentLinearAccScale=<0.01 0.01 0.01>
    overrideWorldGravity {z=-1100}
   }
   n14 = Anim Dynamics - Bone: dyn_hairBang_a_01(ComponentPose=@n13, Alpha=0.3, ExternalForce="0,0,0") {
    angularDampingOverride=1
    angularSpringConstant=1
    bAngularSpring=true
    bGravityOverrideInSimSpace=true
    bOverrideAngularDamping=true
    bUseGravityOverride=true
    alphaBoolBlend {blendOption=Linear}
    boundBone {boneName=dyn_hairBang_a_01}
    componentAppliedLinearAccClamp=<3000 3000 3000>
    componentLinearAccScale=<0.2 0.3 0.2>
    componentLinearVelScale=<0.2 0.3 0.2>
    gravityOverride {z=-900}
    physicsBodyDefinitions [{
      sphereCollisionRadius=0
      boundBone {boneName=dyn_hairBang_a_01}
      boxExtents=<3 5 3>
      constraintSetup {angularLimitsMax=<5 2 10>, angularLimitsMin=<-5 -20 -10>, angularTarget=<0 0 0>}
      localJointOffset=<0.7 0 0>
     }]
    relativeSpaceBone {boneName=head}
   }
   n15 = Anim Dynamics - Bone: dyn_hairBang_a_02(ComponentPose=@n14, Alpha=0.4, ExternalForce="0,0,0") {
    angularDampingOverride=1
    angularSpringConstant=1
    bAngularSpring=true
    bGravityOverrideInSimSpace=true
    bOverrideAngularDamping=true
    bUseGravityOverride=true
    alphaBoolBlend {blendOption=Linear}
    boundBone {boneName=dyn_hairBang_a_02}
    componentAppliedLinearAccClamp=<3000 3000 3000>
    componentLinearAccScale=<0.2 0.3 0.2>
    componentLinearVelScale=<0.2 0.3 0.2>
    gravityOverride {z=-900}
    physicsBodyDefinitions [{
      sphereCollisionRadius=0
      boundBone {boneName=dyn_hairBang_a_02}
      boxExtents=<3 5 3>
      constraintSetup {angularLimitsMax=<10 2 20>, angularLimitsMin=<-10 -60 -20>, angularTarget=<0 0 0>}
      localJointOffset=<0.7 0 0>
     }]
    relativeSpaceBone {boneName=head}
   }
   n16 = Anim Dynamics - Bone: dyn_hairBang_a_03(ComponentPose=@n15, Alpha=0.5, ExternalForce="0,0,0") {
    angularDampingOverride=1
    angularSpringConstant=1
    bAngularSpring=true
    bGravityOverrideInSimSpace=true
    bOverrideAngularDamping=true
    bUseGravityOverride=true
    alphaBoolBlend {blendOption=Linear}
    boundBone {boneName=dyn_hairBang_a_03}
    componentAppliedLinearAccClamp=<3000 3000 3000>
    componentLinearAccScale=<0.2 0.3 0.2>
    componentLinearVelScale=<0.2 0.3 0.2>
    gravityOverride {z=-900}
    physicsBodyDefinitions [{
      sphereCollisionRadius=0
      boundBone {boneName=dyn_hairBang_a_03}
      boxExtents=<3 5 3>
      constraintSetup {angularLimitsMax=<10 2 20>, angularLimitsMin=<-10 -60 -20>, angularTarget=<0 0 0>}
      localJointOffset=<0.7 0 0>
     }]
    relativeSpaceBone {boneName=head}
   }
   n17 = Anim Dynamics - Bone: dyn_hairBang_b_01(ComponentPose=@n16, Alpha=0.3, ExternalForce="0,0,0") {
    angularDampingOverride=1
    angularSpringConstant=1
    bAngularSpring=true
    bGravityOverrideInSimSpace=true
    bOverrideAngularDamping=true
    bUseGravityOverride=true
    alphaBoolBlend {blendOption=Linear}
    boundBone {boneName=dyn_hairBang_b_01}
    componentAppliedLinearAccClamp=<3000 3000 3000>
    componentLinearAccScale=<0.2 0.3 0.2>
    componentLinearVelScale=<0.2 0.3 0.2>
    gravityOverride {z=-900}
    physicsBodyDefinitions [{
      sphereCollisionRadius=0
      boundBone {boneName=dyn_hairBang_b_01}
      boxExtents=<3 5 3>
      constraintSetup {angularLimitsMax=<5 2 10>, angularLimitsMin=<-5 -20 -10>, angularTarget=<0 0 0>}
      localJointOffset=<0.7 0 0>
     }]
    relativeSpaceBone {boneName=head}
   }
   n18 = Anim Dynamics - Bone: dyn_hairBang_b_02(ComponentPose=@n17, Alpha=0.4, ExternalForce="0,0,0") {
    angularDampingOverride=1
    angularSpringConstant=1
    bAngularSpring=true
    bGravityOverrideInSimSpace=true
    bOverrideAngularDamping=true
    bUseGravityOverride=true
    alphaBoolBlend {blendOption=Linear}
    boundBone {boneName=dyn_hairBang_b_02}
    componentAppliedLinearAccClamp=<3000 3000 3000>
    componentLinearAccScale=<0.2 0.3 0.2>
    componentLinearVelScale=<0.2 0.3 0.2>
    gravityOverride {z=-900}
    physicsBodyDefinitions [{
      sphereCollisionRadius=0
      boundBone {boneName=dyn_hairBang_b_02}
      boxExtents=<3 5 3>
      constraintSetup {angularLimitsMax=<10 2 20>, angularLimitsMin=<-10 -60 -20>, angularTarget=<0 0 0>}
      localJointOffset=<0.7 0 0>
     }]
    relativeSpaceBone {boneName=head}
   }
   n19 = Anim Dynamics - Bone: dyn_hairBang_a_03(ComponentPose=@n18, Alpha=0.5, ExternalForce="0,0,0") {
    angularDampingOverride=1
    angularSpringConstant=1
    bAngularSpring=true
    bGravityOverrideInSimSpace=true
    bOverrideAngularDamping=true
    bUseGravityOverride=true
    alphaBoolBlend {blendOption=Linear}
    boundBone {boneName=dyn_hairBang_a_03}
    componentAppliedLinearAccClamp=<3000 3000 3000>
    componentLinearAccScale=<0.2 0.3 0.2>
    componentLinearVelScale=<0.2 0.3 0.2>
    gravityOverride {z=-900}
    physicsBodyDefinitions [{
      sphereCollisionRadius=0
      boundBone {boneName=dyn_hairBang_a_03}
      boxExtents=<3 5 3>
      constraintSetup {angularLimitsMax=<10 2 20>, angularLimitsMin=<-10 -60 -20>, angularTarget=<0 0 0>}
      localJointOffset=<0.7 0 0>
     }]
    relativeSpaceBone {boneName=head}
   }
   n20 = Anim Dynamics - Bone: dyn_hairBang_c_01(ComponentPose=@n19, Alpha=0.3, ExternalForce="0,0,0") {
    angularDampingOverride=1
    angularSpringConstant=1
    bAngularSpring=true
    bGravityOverrideInSimSpace=true
    bOverrideAngularDamping=true
    bUseGravityOverride=true
    alphaBoolBlend {blendOption=Linear}
    boundBone {boneName=dyn_hairBang_c_01}
    componentAppliedLinearAccClamp=<3000 3000 3000>
    componentLinearAccScale=<0.2 0.3 0.2>
    componentLinearVelScale=<0.2 0.3 0.2>
    gravityOverride {z=-900}
    physicsBodyDefinitions [{
      sphereCollisionRadius=0
      boundBone {boneName=dyn_hairBang_c_01}
      boxExtents=<3 5 3>
      constraintSetup {angularLimitsMax=<5 2 10>, angularLimitsMin=<-5 -20 -10>, angularTarget=<0 0 0>}
      localJointOffset=<0.7 0 0>
     }]
    relativeSpaceBone {boneName=head}
   }
   n21 = Anim Dynamics - Bone: dyn_hairBang_c_02(ComponentPose=@n20, Alpha=0.4, ExternalForce="0,0,0") {
    angularDampingOverride=1
    angularSpringConstant=1
    bAngularSpring=true
    bGravityOverrideInSimSpace=true
    bOverrideAngularDamping=true
    bUseGravityOverride=true
    alphaBoolBlend {blendOption=Linear}
    boundBone {boneName=dyn_hairBang_c_02}
    componentAppliedLinearAccClamp=<3000 3000 3000>
    componentLinearAccScale=<0.2 0.3 0.2>
    componentLinearVelScale=<0.2 0.3 0.2>
    gravityOverride {z=-900}
    physicsBodyDefinitions [{
      sphereCollisionRadius=0
      boundBone {boneName=dyn_hairBang_c_02}
      boxExtents=<3 5 3>
      constraintSetup {angularLimitsMax=<10 2 20>, angularLimitsMin=<-10 -60 -20>, angularTarget=<0 0 0>}
      localJointOffset=<0.7 0 0>
     }]
    relativeSpaceBone {boneName=head}
   }
   n22 = Anim Dynamics - Bone: dyn_hairBang_c_03(ComponentPose=@n21, Alpha=0.5, ExternalForce="0,0,0") {
    angularDampingOverride=1
    angularSpringConstant=1
    bAngularSpring=true
    bGravityOverrideInSimSpace=true
    bOverrideAngularDamping=true
    bUseGravityOverride=true
    alphaBoolBlend {blendOption=Linear}
    boundBone {boneName=dyn_hairBang_c_03}
    componentAppliedLinearAccClamp=<3000 3000 3000>
    componentLinearAccScale=<0.2 0.3 0.2>
    componentLinearVelScale=<0.2 0.3 0.2>
    gravityOverride {z=-900}
    physicsBodyDefinitions [{
      sphereCollisionRadius=0
      boundBone {boneName=dyn_hairBang_c_03}
      boxExtents=<3 5 3>
      constraintSetup {angularLimitsMax=<10 2 20>, angularLimitsMin=<-10 -60 -20>, angularTarget=<0 0 0>}
      localJointOffset=<0.7 0 0>
     }]
    relativeSpaceBone {boneName=head}
   }
   n23 = "Constraint - dyn_mainHair_01 (P)"(ComponentPose=@n22, Alpha=0.1, ConstraintWeights_0=1) {
    alphaBoolBlend {blendOption=Linear}
    boneToModify {boneName=dyn_mainHair_01}
    constraintSetup [{targetBone {boneName=clavicle_l}}]
    constraintWeights [1]
   }
   n24 = "Constraint - dyn_mainHair_02 (P)"(ComponentPose=@n23, Alpha=0.1, ConstraintWeights_0=1) {
    alphaBoolBlend {blendOption=Linear}
    boneToModify {boneName=dyn_mainHair_02}
    constraintSetup [{targetBone {boneName=spine_05}}]
    constraintWeights [1]
   }
   n25 = "Constraint - dyn_mainHair_03 (P)"(ComponentPose=@n24, Alpha=0.2, ConstraintWeights_0=1) {
    alphaBoolBlend {blendOption=Linear}
    boneToModify {boneName=dyn_mainHair_03}
    constraintSetup [{targetBone {boneName=spine_05}}]
    constraintWeights [1]
   }
   n26 = "Constraint - dyn_mainHair_04 (P)"(ComponentPose=@n25, Alpha=0.3, ConstraintWeights_0=1) {
    alphaBoolBlend {blendOption=Linear}
    boneToModify {boneName=dyn_mainHair_04}
    constraintSetup [{targetBone {boneName=spine_05}}]
    constraintWeights [1]
   }
   n27 = "Transform (Modify) Bone - Bone: dyn_earing"(ComponentPose=@n26, Alpha=1, Translation="0,0,-0.3", Rotation="5,15,0", Scale="1,1,1") {
    rotationMode=BMM_Additive
    rotationSpace=BCS_BoneSpace
    translationMode=BMM_Additive
    translationSpace=BCS_BoneSpace
    alphaBoolBlend {blendOption=Linear}
    boneToModify {boneName=dyn_earing}
    rotation {pitch=5, yaw=15}
    translation {z=-0.3}
   }
   Output Pose(Result=Component To Local(ComponentPose=@n27)) {layerGroup=None}
  ]}
 EventGraph {nodes [
   Event BlueprintUpdateAnimation()
   > Set Body_Mesh(Body_Mesh=GetOwnerSkeletalMeshComponent())
   TryGetPawnOwner()
   PrintString!(execute, InString="To String (Float)"(InDouble=0), bPrintToScreen, bPrintToLog, TextColor, Duration, Key) -> (then)
   Set NeckLookUp(NeckLookUp:double=GetAlphaFromReferenceTransformDelta(
     SkeletalMeshComponent=GetPartSkeletalMeshComponent()
     BoneName=neck_01
     BaseName=spine_05
     DeltaComponentType=QuaternionTwist
     InputRange=(X=13,Y=15)
     TwistAxis="1,0,0"
    )) -> (Output_Get:double)
  ]}
]
variables [
 Body_Mesh {type=SkeletalMeshComponent}
 NeckLookUp {type=number}
]
